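# Creative island watermark widget that shows translucent creator name and island code text, plus a playtest "Unmoderated" warning.
# Code display events show the name/code row when set to ShowWithWatermark or the viewer is a moderator; playtest-link events toggle the warning line.
# The native CreativePublishWatermark parent supplies the events and inherited text blocks.
# PreConstruct fills placeholder text at design time only when bDesignTime_ApplyDesignTimeValues is false.
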WidgetBlueprint /FortniteCoreUI/InGame/Creative/CreativePublishWatermark
graphs [EventGraph {nodes [
   n1 = Event PreConstruct()
   > Branch(Condition=@n1.IsDesignTime)
   > n2 = Branch(Condition=Get bDesignTime_ApplyDesignTimeValues() -> (bDesignTime_ApplyDesignTimeValues:bool))
   "SetText (Text)"(execute=@n2.else, self=Get Text_CreatorName() -> (Text_CreatorName:Common Text Object Reference), InText="INVTEXT(\"{Creator Name}\")")
   > "SetText (Text)"(self=Get Text_Mnemonic() -> (Text_Mnemonic:Common Text Object Reference), InText="INVTEXT(\"{0123-4567-8910}\")")
   n3 = Event OnPublishedIslandCodeDisplayChanged!() -> (OutputDelegate, then, PublishedIslandCodeDisplay:ECreativePublishedIslandCodeDisplay Enum=Off) {is=title K2Node_Event}
   n4 = IsModerator!(self:Creative Publish Watermark Object Reference) -> (ReturnValue:bool)
   n5 = "Equal (Enum)"(A:ECreativePublishedIslandCodeDisplay Enum=@n3.PublishedIslandCodeDisplay, B:ECreativePublishedIslandCodeDisplay Enum=ShowWithWatermark)
   > SetVisibility(
    self=Get HBox_Container() -> (HBox_Container:Horizontal Box Object Reference)
    InVisibility=Select(Option 0=Collapsed, Option 1=SelfHitTestInvisible, Index=OR Boolean(A=@n5, B=@n4))
   )
   n6 = Event OnIsPlaytestLinkChanged!() -> (OutputDelegate, then, bIsPlaytestLink:bool) {is=title K2Node_Event}
   > SetVisibility(
    self=Get Text_Playtest() -> (Text_Playtest:Common Text Object Reference)
    InVisibility=Select(Option 0=Collapsed, Option 1=SelfHitTestInvisible, Index=@n6.bIsPlaytestLink)
   )
  ]}]
variables []
widgets {
 info {inheritedWidgetCount=3, parentClass=/Script/FortniteUI.CreativePublishWatermark, rootWidgetClass=&VerticalBox, widgetCount=7}
 widgets [
  VerticalBox_0 {widgetClassPath=&VerticalBox, widget=VerticalBox_0}
  HBox_Container {
   widgetClassPath=&HorizontalBox
   bIsVariable=true
   parent=VerticalBox_0
   widget=HBox_Container
   slot {class=VerticalBoxSlot, properties {horizontalAlignment=HAlign_Center}}
  }
  Text_CreatorName {
   widgetClassPath=&CommonTextBlock
   bInherited=true
   parent=HBox_Container
   widget=Text_CreatorName
   properties {
    bAutoCollapseWithEmptyText=true
    justification=Center
    colorAndOpacity {colorUseRule=UseColor_Specified, specifiedColor=<rgba 1 1 1 0.502>}
    font {
     bForceMonospaced=false
     bMaterialIsStencil=false
     fontMaterial=null
     fontObject=&NotoSans
     letterSpacing=0
     monospacedWidth=1
     size=18
     skewAmount=0
     typefaceFontName=NotoSans-Bold
     outlineSettings {
      bApplyOutlineToDropShadows=false
      bMiteredCorners=false
      bSeparateFillAlpha=false
      outlineMaterial=null
      outlineSize=0
      outlineColor=<rgba 0 0 0 1>
     }
    }
    shadowOffset=<0 0>
   }
   slot {class=HorizontalBoxSlot, properties {padding=<ltrb 0 10 0 0>}}
  }
  Spacer_53 {widgetClassPath=&Spacer, parent=HBox_Container, widget=Spacer_53, properties {size=<20 1>}, slot {class=HorizontalBoxSlot}}
  Text_Mnemonic {
   widgetClassPath=&CommonTextBlock
   bInherited=true
   parent=HBox_Container
   widget=Text_Mnemonic
   properties {
    bAutoCollapseWithEmptyText=true
    justification=Center
    colorAndOpacity {colorUseRule=UseColor_Specified, specifiedColor=<rgba 1 1 1 0.502>}
    font {
     bForceMonospaced=false
     bMaterialIsStencil=false
     fontMaterial=null
     fontObject=&NotoSans
     letterSpacing=0
     monospacedWidth=1
     size=18
     skewAmount=0
     typefaceFontName=NotoSans-Bold
     outlineSettings {
      bApplyOutlineToDropShadows=false
      bMiteredCorners=false
      bSeparateFillAlpha=false
      outlineMaterial=null
      outlineSize=0
      outlineColor=<rgba 0 0 0 1>
     }
    }
    shadowOffset=<0 0>
   }
   slot {class=HorizontalBoxSlot, properties {padding=<ltrb 0 10 0 0>}}
  }
  Text_Playtest {
   widgetClassPath=&CommonTextBlock
   bIsVariable=true
   parent=VerticalBox_0
   widget=Text_Playtest
   properties {
    justification=Center
    text="Playtest (Warning: Unmoderated)"
    textTransformPolicy=ToUpper
    visibility=Collapsed
    colorAndOpacity {colorUseRule=UseColor_Specified, specifiedColor=<rgba 0.9301 1 0.0091 0.8>}
    font {
     bForceMonospaced=false
     bMaterialIsStencil=false
     fontMaterial=null
     fontObject=&NotoSans
     letterSpacing=0
     monospacedWidth=1
     size=18
     skewAmount=0
     typefaceFontName=NotoSans-Bold
     outlineSettings {
      bApplyOutlineToDropShadows=false
      bMiteredCorners=false
      bSeparateFillAlpha=false
      outlineMaterial=null
      outlineSize=0
      outlineColor=<rgba 0 0 0 1>
     }
    }
    shadowOffset=<0 0>
   }
   slot {class=VerticalBoxSlot, properties {horizontalAlignment=HAlign_Center}}
  }
  TutorialHighlight {widgetClassPath=&TutorialHighlight}
 ]
}
