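# Creative-mode "Publish Island" modal, built on FortCreativePublishModal, with title and description fields and a copyable island code.
# Parent events toggle a spinner that disables Cancel, and failed validation or publish errors play a shake animation.
# On construct it shows the clear-player-data row if persistence devices exist and disables both text fields in cabin mode.
# Cancel, Back and touch close deactivate the widget, while Copy Code plays a fading "Copied" animation.
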WidgetBlueprint /FortniteCoreUI/InGame/Creative/PublishModal
animations [
 Error {
  binding_count=4
  length_seconds=0.25
  object=&self:Error
  sequence {bindings [
    Text_ErrorInDescriptionField {
     id=#1
     possessed_class=CommonTextBlock
     tracks [{
       type=MovieScene2DTransformTrack
       sections [{
         type=MovieScene2DTransformSection
         end_frame=5
         start_frame=0
         channels [
          Translation.X {type=MovieSceneScriptingFloatChannel, keys [[0,0],[1,-20],[2,20],[3,-10],[4,10],[5,0]]}
          Translation.Y {type=MovieSceneScriptingFloatChannel}
          Angle {type=MovieSceneScriptingFloatChannel}
          Scale.X {type=MovieSceneScriptingFloatChannel}
          Scale.Y {type=MovieSceneScriptingFloatChannel}
          Shear.X {type=MovieSceneScriptingFloatChannel}
          Shear.Y {type=MovieSceneScriptingFloatChannel}
         ]
        }]
      }]
    }
    Button_Submit {
     id=#2
     possessed_class=PrimarySkewButton_C
     tracks [{
       type=MovieScene2DTransformTrack
       sections [{
         type=MovieScene2DTransformSection
         end_frame=5
         start_frame=0
         channels [
          Translation.X {type=MovieSceneScriptingFloatChannel, keys [[0,0],[1,20],[2,-20],[3,10],[4,-10],[5,0]]}
          Translation.Y {type=MovieSceneScriptingFloatChannel, const [0,1]}
          Angle {type=MovieSceneScriptingFloatChannel}
          Scale.X {type=MovieSceneScriptingFloatChannel}
          Scale.Y {type=MovieSceneScriptingFloatChannel}
          Shear.X {type=MovieSceneScriptingFloatChannel}
          Shear.Y {type=MovieSceneScriptingFloatChannel}
         ]
        }]
      }]
    }
    Text_ErrorInNameField {
     id=#3
     possessed_class=CommonTextBlock
     tracks [{
       type=MovieScene2DTransformTrack
       sections [{
         type=MovieScene2DTransformSection
         end_frame=5
         start_frame=0
         channels [
          Translation.X {type=MovieSceneScriptingFloatChannel, keys [[0,0],[1,20],[2,-20],[3,10],[4,-10],[5,0]]}
          Translation.Y {type=MovieSceneScriptingFloatChannel}
          Angle {type=MovieSceneScriptingFloatChannel}
          Scale.X {type=MovieSceneScriptingFloatChannel}
          Scale.Y {type=MovieSceneScriptingFloatChannel}
          Shear.X {type=MovieSceneScriptingFloatChannel}
          Shear.Y {type=MovieSceneScriptingFloatChannel}
         ]
        }]
      }]
    }
    Text_ErrorOnPublish {
     id=#4
     possessed_class=CommonTextBlock
     tracks [{
       type=MovieScene2DTransformTrack
       sections [{
         type=MovieScene2DTransformSection
         end_frame=5
         start_frame=1
         channels [
          Translation.X {type=MovieSceneScriptingFloatChannel, keys [[1,20],[2,-20],[3,10],[4,-10],[5,0]]}
          Translation.Y {type=MovieSceneScriptingFloatChannel}
          Angle {type=MovieSceneScriptingFloatChannel}
          Scale.X {type=MovieSceneScriptingFloatChannel}
          Scale.Y {type=MovieSceneScriptingFloatChannel}
          Shear.X {type=MovieSceneScriptingFloatChannel}
          Shear.Y {type=MovieSceneScriptingFloatChannel}
         ]
        }]
      }]
    }
   ]}
 }
 Copied {
  binding_count=1
  length_seconds=2.35
  object=&self:Copied
  sequence {bindings [CommonTextBlock_8 {
     id=#5
     possessed_class=CommonTextBlock
     tracks [
      {
       type=MovieSceneByteTrack
       display_name=Visibility
       sections [{type=MovieSceneByteSection, end_frame=1, start_frame=0, channels [None {type=MovieSceneScriptingByteChannel, const [4,1]}]}]
      }
      {
       type=MovieScene2DTransformTrack
       sections [{
         type=MovieScene2DTransformSection
         end_frame=47
         start_frame=1
         channels [
          Translation.X {type=MovieSceneScriptingFloatChannel}
          Translation.Y {type=MovieSceneScriptingFloatChannel}
          Angle {type=MovieSceneScriptingFloatChannel}
          Scale.X {type=MovieSceneScriptingFloatChannel, keys [[1,0],[4,1.2],[6,1]]}
          Scale.Y {type=MovieSceneScriptingFloatChannel, keys [[1,0],[4,1.2],[6,1]]}
          Shear.X {type=MovieSceneScriptingFloatChannel}
          Shear.Y {type=MovieSceneScriptingFloatChannel}
         ]
        }]
      }
      {
       type=MovieSceneFloatTrack
       display_name=Render Opacity
       sections [{
         type=MovieSceneFloatSection
         end_frame=47
         start_frame=0
         channels [None {type=MovieSceneScriptingFloatChannel, keys [[1,0],[4,1],[31,1],[47,0]]}]
        }]
      }
     ]
    }]}
 }
]
graphs [
 EventGraph {nodes [
   n1 = "On Button Clicked (Button_Cancel)"!() -> (OutputDelegate, then, Button:Common Button Legacy Object Reference) {is=title K2Node_ComponentBoundEvent}
   DeactivateWidget(execute=[@n1.then, @n2.then, @n3.then, @n4.then])
   "On Button Clicked (Button_CopyCode)"!() -> (OutputDelegate, then, Button:Common Button Legacy Object Reference) {is=title K2Node_ComponentBoundEvent}
   > PlayAnimation(InAnimation=Get Copied() -> (Copied:Widget Animation Object Reference))
   n2 = "On Button Clicked (Button_Back)"!() -> (OutputDelegate, then, Button:Common Button Legacy Object Reference) {is=title K2Node_ComponentBoundEvent}
   n3 = "On Button Clicked (Button_CloseTouch)"!() -> (OutputDelegate, then, Button:Common Button Legacy Object Reference) {is=title K2Node_ComponentBoundEvent}
   n4 = Close() {is=title K2Node_CustomEvent}
   Event Reset() {is=title K2Node_Event}
   > ShowSpinner!(execute, self, bShowSpinner:bool=false) -> (then)
   n5 = Event OnValidCheckComplete!() -> (OutputDelegate, then, bSuccess:bool) {is=title K2Node_Event}
   > n6 = Branch(Condition=@n5.bSuccess)
   PlayAnimation(execute=@n6.else, InAnimation=Get Error() -> (Error:Widget Animation Object Reference))
   Event OnPublishError!() -> (OutputDelegate, then, ErrorText:"Text (by ref)") {is=title K2Node_Event}
   > PlayAnimation(InAnimation=Get Error())
   > ShowSpinner(bShowSpinner=false)
   Event OnPublishBegin() {is=title K2Node_Event}
   > ShowSpinner(bShowSpinner=true)
   Event Construct()
   > n7 = Sequence()
   > n8 = Cast To FortGameStateAthena!(execute, Object=GetGameState!() -> (ReturnValue:Game State Base Object Reference)) -> (then, CastFailed, AsFort Game State Athena:Fort Game State Athena Object Reference) {is=title K2Node_DynamicCast}
   n9 = GetOwningPlayer()
   > Is Valid(InputObject=@n9)
   n10 = Get Controlled Pawn!(self:Controller Object Reference=@n9) -> (ReturnValue:Pawn Object Reference) {is=title K2Node_CallFunction}
   > SetVisibility(
    self=Get HBox_ClearAllPlayersData() -> (HBox_ClearAllPlayersData:Horizontal Box Object Reference)
    InVisibility=Select(
     Option 0=Collapsed
     Option 1=SelfHitTestInvisible
     Index=IsTherePersistenceDevices!(self:Fort Game State Athena Object Reference=@n8, SwitchingPawn:Pawn Object Reference=@n10) -> (ReturnValue:bool)
    )
   )
   Branch(execute=@n7.then_1, Condition=IsInCabinMode!() -> (ReturnValue:bool))
   > SetIsEnabled(
    self=[
     Get EditText_Name() -> (EditText_Name:Text Box Object Reference)
     Get EditText_Description() -> (EditText_Description:Text Box Object Reference)
    ]
    bInIsEnabled=false
   )
  ]}
 ShowSpinner {nodes [
   n1 = ShowSpinner!() -> (then, bShowSpinner:bool=~) {is=title K2Node_FunctionEntry}
   > SetVisibility(
    self=Get Image_Spinner() -> (Image_Spinner:Image Object Reference)
    InVisibility=Select(Option 0=Collapsed, Option 1=SelfHitTestInvisible, Index=@n1)
   )
   > SetIsEnabled(self=Get Button_Cancel() -> (Button_Cancel:Icon Text Button Object Reference), bInIsEnabled=NOT Boolean(A=@n1))
  ]}
]
variables []
widgets {
 info {
  inheritedWidgetCount=15
  namedSlotCount=1
  parentClass=/Script/FortniteUI.FortCreativePublishModal
  rootWidgetClass=&Border
  widgetCount=50
 }
 widgets [
  Border_5 {widgetClassPath=&Border, widget=Border_5, properties {brushColor=<rgba 0 0 0 0.2>, padding=<ltrb 0 0 0 0>}}
  Lightbox {
   widgetClassPath=/Game/UI/Foundation/Lightbox/Lightbox
   parent=Border_5
   widget=Lightbox
   properties {dimAmount=0.7}
   slot {class=BorderSlot, properties {padding=<ltrb 0 0 0 0>}}
  }
  BackgroundBlur_0 {
   widgetClassPath=&BackgroundBlur
   namedSlotHost=Lightbox
   widget=BackgroundBlur_0
   properties {
    blurStrength=30
    horizontalAlignment=HAlign_Center
    verticalAlignment=VAlign_Center
    lowQualityFallbackBrush {
     bIsDynamicallyLoaded=false
     drawAs=NoDrawType
     imageType=NoImage
     mirroring=NoMirror
     resourceName=None
     resourceObject=null
     tiling=NoTile
     imageSize=<0 0>
     margin=<ltrb 0 0 0 0>
     outlineSettings {
      bUseBrushTransparency=false
      roundingType=HalfHeightRadius
      width=0
      color {colorUseRule=UseColor_Specified, specifiedColor=<rgba 0 0 0 0>}
      cornerRadii=<0 0 0 0>
     }
     tintColor {colorUseRule=UseColor_Specified, specifiedColor=<rgba 0 0 0 0.54>}
     uVRegion {bIsValid=false, max=<0 0>, min=<0 0>}
    }
   }
  }
  SizeBox_0 {
   widgetClassPath=&SizeBox
   parent=BackgroundBlur_0
   widget=SizeBox_0
   properties {bOverride_MinDesiredWidth=true, heightOverride=950, maxDesiredHeight=1000, minDesiredWidth=1300, widthOverride=1300}
   slot {class=BackgroundBlurSlot, properties {horizontalAlignment=HAlign_Center, verticalAlignment=VAlign_Center, padding=<ltrb 0 0 0 0>}}
  }
  VerticalBox_0 {widgetClassPath=&VerticalBox, parent=SizeBox_0, widget=VerticalBox_0, slot {class=SizeBoxSlot}}
  Border_0 {
   widgetClassPath=&Border
   parent=VerticalBox_0
   widget=Border_0
   properties {
    verticalAlignment=VAlign_Top
    background {
     bIsDynamicallyLoaded=false
     drawAs=Image
     imageType=NoImage
     mirroring=NoMirror
     resourceName=None
     resourceObject=null
     tiling=NoTile
     imageSize=<32 32>
     margin=<ltrb 0 0 0 0>
     outlineSettings {
      bUseBrushTransparency=false
      roundingType=HalfHeightRadius
      width=0
      color {colorUseRule=UseColor_Specified, specifiedColor=<rgba 0 0 0 0>}
      cornerRadii=<0 0 0 0>
     }
     tintColor {colorUseRule=UseColor_Specified, specifiedColor=<rgba 0 0.0078 0.0431 1>}
     uVRegion {bIsValid=false, max=<0 0>, min=<0 0>}
    }
    padding=<ltrb 0 0 0 0>
   }
   slot {class=VerticalBoxSlot}
  }
  Overlay_0 {
   widgetClassPath=&Overlay
   parent=Border_0
   widget=Overlay_0
   slot {class=BorderSlot, properties {verticalAlignment=VAlign_Top, padding=<ltrb 0 0 0 0>}}
  }
  VerticalBox_1 {
   widgetClassPath=&VerticalBox
   parent=Overlay_0
   widget=VerticalBox_1
   slot {class=OverlaySlot, properties {horizontalAlignment=HAlign_Center}}
  }
  TextBlock_0 {
   widgetClassPath=&TextBlock
   parent=VerticalBox_1
   widget=TextBlock_0
   properties {
    justification=Center
    text=PUBLISH ISLAND
    font {
     bForceMonospaced=false
     bMaterialIsStencil=false
     fontMaterial=null
     fontObject=&Burbank
     letterSpacing=0
     monospacedWidth=1
     size=101
     skewAmount=0
     typefaceFontName=Burbank-Black
     outlineSettings {
      bApplyOutlineToDropShadows=false
      bMiteredCorners=false
      bSeparateFillAlpha=false
      outlineMaterial=null
      outlineSize=0
      outlineColor=<rgba 0 0 0 1>
     }
    }
    renderTransform {angle=0, scale=<1 1>, shear=<0 0>, translation=<0 -56>}
   }
   slot {class=VerticalBoxSlot, properties {padding=<ltrb 0 0 0 -70>}}
  }
  CommonTextBlock_1 {
   widgetClassPath=&CommonTextBlock
   parent=VerticalBox_1
   widget=CommonTextBlock_1
   properties {
    justification=Center
    text=Publishing will make a playable copy of your island that anyone can visit using the provided code.
    wrapTextAt=1060.9664
    colorAndOpacity {colorUseRule=UseColor_Specified, specifiedColor=<rgba 0.8851 0.9301 0.8779 1>}
    font {
     bForceMonospaced=false
     bMaterialIsStencil=false
     fontMaterial=null
     fontObject=&BurbankSmall
     letterSpacing=0
     monospacedWidth=1
     size=24
     skewAmount=0
     typefaceFontName=Black
     outlineSettings {
      bApplyOutlineToDropShadows=false
      bMiteredCorners=false
      bSeparateFillAlpha=false
      outlineMaterial=null
      outlineSize=0
      outlineColor=<rgba 0 0 0 1>
     }
    }
    shadowOffset=<0 0>
   }
   slot {class=VerticalBoxSlot, properties {horizontalAlignment=HAlign_Center, padding=<ltrb 0 0 0 10>}}
  }
  Text_ErrorOnPublish {
   widgetClassPath=&CommonTextBlock
   bInherited=true
   parent=VerticalBox_1
   widget=Text_ErrorOnPublish
   properties {
    autoWrapText=true
    bAutoCollapseWithEmptyText=true
    justification=Center
    text="{PUBLISHING ERROR...}"
    visibility=Collapsed
    colorAndOpacity {colorUseRule=UseColor_Specified, specifiedColor=<rgba 0.9301 0.0742 0.107 1>}
    font {
     bForceMonospaced=false
     bMaterialIsStencil=false
     fontMaterial=null
     fontObject=&BurbankSmall
     letterSpacing=0
     monospacedWidth=1
     size=24
     skewAmount=0
     typefaceFontName=Black
     outlineSettings {
      bApplyOutlineToDropShadows=false
      bMiteredCorners=false
      bSeparateFillAlpha=false
      outlineMaterial=null
      outlineSize=0
      outlineColor=<rgba 0 0 0 1>
     }
    }
    shadowOffset=<0 0>
   }
   slot {class=VerticalBoxSlot, properties {verticalAlignment=VAlign_Center, padding=<ltrb 0 10 0 10>}}
  }
  VisibilityBorder_Touch {
   widgetClassPath=&CommonHardwareVisibilityBorder
   parent=Overlay_0
   widget=VisibilityBorder_Touch
   properties {
    renderTransform {angle=0, scale=<1 1>, shear=<0 0>, translation=<30 -20>}
    visibilityQuery {
     autoDescription=" ANY( Input.Touch )"
     tokenStreamVersion=0
     userDescription=""
     queryTokenStream [0,1,1,1,0]
     tagDictionary [{tagName=Input.Touch}]
    }
   }
   slot {class=OverlaySlot, properties {horizontalAlignment=HAlign_Right}}
  }
  Button_CloseTouch {
   widgetClassPath=/FortUILibrary/Foundation/Buttons/CloseButton
   bIsVariable=true
   parent=VisibilityBorder_Touch
   widget=Button_CloseTouch
   properties {
    onButtonBaseClicked=(&self:WidgetTree.Button_CloseTouch.HandleOnButtonClicked)
    onButtonBaseDoubleClicked=(&self:WidgetTree.Button_CloseTouch.HandleOnButtonDoubleClicked)
    onButtonBaseHovered=(&self:WidgetTree.Button_CloseTouch.HandleOnButtonHovered)
    onButtonBaseUnhovered=(&self:WidgetTree.Button_CloseTouch.HandleOnButtonUnhovered)
    onSelectedChangedBase=(&self:WidgetTree.Button_CloseTouch.HandleOnSelectedChanged)
   }
   slot {class=BorderSlot, properties {padding=<ltrb 0 0 0 0>}}
  }
  CommonBorder_2 {
   widgetClassPath=&CommonBorder
   parent=VerticalBox_0
   widget=CommonBorder_2
   properties {
    background {
     bIsDynamicallyLoaded=false
     drawAs=Image
     imageType=NoImage
     mirroring=NoMirror
     resourceName=None
     resourceObject=null
     tiling=NoTile
     imageSize=<32 32>
     margin=<ltrb 0 0 0 0>
     outlineSettings {
      bUseBrushTransparency=false
      roundingType=HalfHeightRadius
      width=0
      color {colorUseRule=UseColor_Specified, specifiedColor=<rgba 0 0 0 0>}
      cornerRadii=<0 0 0 0>
     }
     tintColor {colorUseRule=UseColor_Specified, specifiedColor=<rgba 0 0.011 0.0469 1>}
     uVRegion {bIsValid=false, max=<0 0>, min=<0 0>}
    }
    padding=<ltrb 20 20 20 0>
   }
   slot {class=VerticalBoxSlot, properties {size {sizeRule=Fill, value=1}}}
  }
  VerticalBox_6 {
   widgetClassPath=&VerticalBox
   parent=CommonBorder_2
   widget=VerticalBox_6
   slot {class=BorderSlot, properties {padding=<ltrb 20 20 20 0>}}
  }
  Text_IslandCode {
   widgetClassPath=&CommonTextBlock
   bInherited=true
   parent=VerticalBox_6
   widget=Text_IslandCode
   properties {
    bAutoCollapseWithEmptyText=true
    justification=Center
    text="{Published Code}"
    visibility=Collapsed
    font {
     bForceMonospaced=false
     bMaterialIsStencil=false
     fontMaterial=/FortUILibrary/Materials/Foundation/M_Font_Horizon_Dynamic
     fontObject=&BurbankSmall
     letterSpacing=0
     monospacedWidth=1
     size=28
     skewAmount=0
     typefaceFontName=Black
     outlineSettings {
      bApplyOutlineToDropShadows=false
      bMiteredCorners=false
      bSeparateFillAlpha=false
      outlineMaterial=null
      outlineSize=0
      outlineColor=<rgba 0 0 0 1>
     }
    }
    shadowOffset=<0 0>
   }
   slot {class=VerticalBoxSlot, properties {padding=<ltrb 10 0 10 20>}}
  }
  Switcher_Content {
   widgetClassPath=/Script/CommonUI.CommonAnimatedSwitcher
   bInherited=true
   bIsVariable=true
   parent=VerticalBox_6
   widget=Switcher_Content
   slot {class=VerticalBoxSlot}
  }
  VB_PublishEntry {
   widgetClassPath=&VerticalBox
   parent=Switcher_Content
   widget=VB_PublishEntry
   properties {navigation=&self:WidgetTree.VB_PublishEntry.WidgetNavigation_0}
   slot {class=WidgetSwitcherSlot}
  }
  HorizontalBox_2 {
   widgetClassPath=&HorizontalBox
   parent=VB_PublishEntry
   widget=HorizontalBox_2
   slot {class=VerticalBoxSlot, properties {verticalAlignment=VAlign_Top}}
  }
  CommonTextBlock_3 {
   widgetClassPath=&CommonTextBlock
   parent=HorizontalBox_2
   widget=CommonTextBlock_3
   properties {
    justification=Right
    text=TITLE:
    font {
     bForceMonospaced=false
     bMaterialIsStencil=false
     fontMaterial=null
     fontObject=&BurbankSmall
     letterSpacing=0
     monospacedWidth=1
     size=36
     skewAmount=0
     typefaceFontName=Black
     outlineSettings {
      bApplyOutlineToDropShadows=false
      bMiteredCorners=false
      bSeparateFillAlpha=false
      outlineMaterial=null
      outlineSize=0
      outlineColor=<rgba 0 0 0 1>
     }
    }
    shadowOffset=<0 0>
   }
   slot {class=HorizontalBoxSlot, properties {verticalAlignment=VAlign_Top, padding=<ltrb 0 15 20 0>, size {sizeRule=Fill, value=1}}}
  }
  VerticalBox_2 {widgetClassPath=&VerticalBox, parent=HorizontalBox_2, widget=VerticalBox_2, slot {class=HorizontalBoxSlot}}
  VerticalBox_5 {
   widgetClassPath=&VerticalBox
   parent=VerticalBox_2
   widget=VerticalBox_5
   slot {class=VerticalBoxSlot, properties {horizontalAlignment=HAlign_Center, padding=<ltrb 0 15 0 0>}}
  }
  HorizontalBox_4 {widgetClassPath=&HorizontalBox, parent=VerticalBox_5, widget=HorizontalBox_4, slot {class=VerticalBoxSlot}}
  SizeBox_4 {
   widgetClassPath=&SizeBox
   parent=HorizontalBox_4
   widget=SizeBox_4
   properties {bOverride_HeightOverride=true, bOverride_WidthOverride=true, heightOverride=60, widthOverride=1000}
   slot {class=HorizontalBoxSlot, properties {size {sizeRule=Fill, value=1}}}
  }
  EditText_Name {
   widgetClassPath=&EditableTextBox
   bInherited=true
   bIsVariable=true
   parent=SizeBox_4
   widget=EditText_Name
   properties {
    hintText="Your Island's name goes here!"
    justification=Center
    navigation=&self:WidgetTree.EditText_Name.WidgetNavigation_0
    widgetStyle {
     backgroundColor {colorUseRule=UseColor_Specified, specifiedColor=<rgba 1 1 1 1>}
     backgroundImageFocused {
      bIsDynamicallyLoaded=false
      drawAs=Box
      imageType=FullColor
      mirroring=NoMirror
      resourceName=../../../Engine/Content/Slate/Common/TextBox_Hovered.png
      resourceObject=null
      tiling=NoTile
      imageSize=<0 0>
      margin=<ltrb 0.25 0.25 0.25 0.25>
      outlineSettings {
       bUseBrushTransparency=false
       roundingType=HalfHeightRadius
       width=0
       color {colorUseRule=UseColor_Specified, specifiedColor=<rgba 0 0 0 0>}
       cornerRadii=<0 0 0 0>
      }
      tintColor {colorUseRule=UseColor_Specified, specifiedColor=<rgba 1 1 1 1>}
      uVRegion {bIsValid=false, max=<0 0>, min=<0 0>}
     }
     backgroundImageHovered {
      bIsDynamicallyLoaded=false
      drawAs=Box
      imageType=FullColor
      mirroring=NoMirror
      resourceName=../../../Engine/Content/Slate/Common/TextBox_Hovered.png
      resourceObject=null
      tiling=NoTile
      imageSize=<0 0>
      margin=<ltrb 0.25 0.25 0.25 0.25>
      outlineSettings {
       bUseBrushTransparency=false
       roundingType=HalfHeightRadius
       width=0
       color {colorUseRule=UseColor_Specified, specifiedColor=<rgba 0 0 0 0>}
       cornerRadii=<0 0 0 0>
      }
      tintColor {colorUseRule=UseColor_Specified, specifiedColor=<rgba 1 1 1 1>}
      uVRegion {bIsValid=false, max=<0 0>, min=<0 0>}
     }
     backgroundImageNormal {
      bIsDynamicallyLoaded=false
      drawAs=Box
      imageType=FullColor
      mirroring=NoMirror
      resourceName=../../../Engine/Content/Slate/Common/TextBox.png
      resourceObject=null
      tiling=NoTile
      imageSize=<0 0>
      margin=<ltrb 0.25 0.25 0.25 0.25>
      outlineSettings {
       bUseBrushTransparency=false
       roundingType=HalfHeightRadius
       width=0
       color {colorUseRule=UseColor_Specified, specifiedColor=<rgba 0 0 0 0>}
       cornerRadii=<0 0 0 0>
      }
      tintColor {colorUseRule=UseColor_Specified, specifiedColor=<rgba 1 1 1 1>}
      uVRegion {bIsValid=false, max=<0 0>, min=<0 0>}
     }
     backgroundImageReadOnly {
      bIsDynamicallyLoaded=false
      drawAs=Box
      imageType=FullColor
      mirroring=NoMirror
      resourceName=../../../Engine/Content/Slate/Common/TextBox_ReadOnly.png
      resourceObject=null
      tiling=NoTile
      imageSize=<0 0>
      margin=<ltrb 0.25 0.25 0.25 0.25>
      outlineSettings {
       bUseBrushTransparency=false
       roundingType=HalfHeightRadius
       width=0
       color {colorUseRule=UseColor_Specified, specifiedColor=<rgba 0 0 0 0>}
       cornerRadii=<0 0 0 0>
      }
      tintColor {colorUseRule=UseColor_Specified, specifiedColor=<rgba 1 1 1 1>}
      uVRegion {bIsValid=false, max=<0 0>, min=<0 0>}
     }
     focusedForegroundColor {colorUseRule=UseColor_Foreground, specifiedColor=<rgba 1 0 1 1>}
     foregroundColor {colorUseRule=UseColor_Specified, specifiedColor=<rgba 0 0 0 1>}
     hScrollBarPadding=<ltrb 0 0 0 0>
     padding=<ltrb 6 10 6 10>
     readOnlyForegroundColor {colorUseRule=UseColor_Foreground, specifiedColor=<rgba 1 0 1 1>}
     scrollBarStyle {
      thickness=16
      draggedThumbImage {
       bIsDynamicallyLoaded=false
       drawAs=Box
       imageType=FullColor
       mirroring=NoMirror
       resourceName=../../../Engine/Content/Slate/Common/Scrollbar_Thumb.png
       resourceObject=null
       tiling=NoTile
       imageSize=<0 0>
       margin=<ltrb 0.25 0.25 0.25 0.25>
       outlineSettings {
        bUseBrushTransparency=false
        roundingType=HalfHeightRadius
        width=0
        color {colorUseRule=UseColor_Specified, specifiedColor=<rgba 0 0 0 0>}
        cornerRadii=<0 0 0 0>
       }
       tintColor {colorUseRule=UseColor_Specified, specifiedColor=<rgba 1 1 1 1>}
       uVRegion {bIsValid=false, max=<0 0>, min=<0 0>}
      }
      horizontalBackgroundImage {
       bIsDynamicallyLoaded=false
       drawAs=NoDrawType
       imageType=NoImage
       mirroring=NoMirror
       resourceName=None
       resourceObject=null
       tiling=NoTile
       imageSize=<0 0>
       margin=<ltrb 0 0 0 0>
       outlineSettings {
        bUseBrushTransparency=false
        roundingType=HalfHeightRadius
        width=0
        color {colorUseRule=UseColor_Specified, specifiedColor=<rgba 0 0 0 0>}
        cornerRadii=<0 0 0 0>
       }
       tintColor {colorUseRule=UseColor_Specified, specifiedColor=<rgba 1 1 1 1>}
       uVRegion {bIsValid=false, max=<0 0>, min=<0 0>}
      }
      horizontalBottomSlotImage {
       bIsDynamicallyLoaded=false
       drawAs=Image
       imageType=FullColor
       mirroring=NoMirror
       resourceName=../../../Engine/Content/Slate/Common/Scrollbar_Background_Horizontal.png
       resourceObject=null
       tiling=NoTile
       imageSize=<8 8>
       margin=<ltrb 0 0 0 0>
       outlineSettings {
        bUseBrushTransparency=false
        roundingType=HalfHeightRadius
        width=0
        color {colorUseRule=UseColor_Specified, specifiedColor=<rgba 0 0 0 0>}
        cornerRadii=<0 0 0 0>
       }
       tintColor {colorUseRule=UseColor_Specified, specifiedColor=<rgba 1 1 1 1>}
       uVRegion {bIsValid=false, max=<0 0>, min=<0 0>}
      }
      horizontalTopSlotImage {
       bIsDynamicallyLoaded=false
       drawAs=Image
       imageType=FullColor
       mirroring=NoMirror
       resourceName=../../../Engine/Content/Slate/Common/Scrollbar_Background_Horizontal.png
       resourceObject=null
       tiling=NoTile
       imageSize=<8 8>
       margin=<ltrb 0 0 0 0>
       outlineSettings {
        bUseBrushTransparency=false
        roundingType=HalfHeightRadius
        width=0
        color {colorUseRule=UseColor_Specified, specifiedColor=<rgba 0 0 0 0>}
        cornerRadii=<0 0 0 0>
       }
       tintColor {colorUseRule=UseColor_Specified, specifiedColor=<rgba 1 1 1 1>}
       uVRegion {bIsValid=false, max=<0 0>, min=<0 0>}
      }
      hoveredThumbImage {
       bIsDynamicallyLoaded=false
       drawAs=Box
       imageType=FullColor
       mirroring=NoMirror
       resourceName=../../../Engine/Content/Slate/Common/Scrollbar_Thumb.png
       resourceObject=null
       tiling=NoTile
       imageSize=<0 0>
       margin=<ltrb 0.25 0.25 0.25 0.25>
       outlineSettings {
        bUseBrushTransparency=false
        roundingType=HalfHeightRadius
        width=0
        color {colorUseRule=UseColor_Specified, specifiedColor=<rgba 0 0 0 0>}
        cornerRadii=<0 0 0 0>
       }
       tintColor {colorUseRule=UseColor_Specified, specifiedColor=<rgba 1 1 1 1>}
       uVRegion {bIsValid=false, max=<0 0>, min=<0 0>}
      }
      normalThumbImage {
       bIsDynamicallyLoaded=false
       drawAs=Box
       imageType=FullColor
       mirroring=NoMirror
       resourceName=../../../Engine/Content/Slate/Common/Scrollbar_Thumb.png
       resourceObject=null
       tiling=NoTile
       imageSize=<0 0>
       margin=<ltrb 0.25 0.25 0.25 0.25>
       outlineSettings {
        bUseBrushTransparency=false
        roundingType=HalfHeightRadius
        width=0
        color {colorUseRule=UseColor_Specified, specifiedColor=<rgba 0 0 0 0>}
        cornerRadii=<0 0 0 0>
       }
       tintColor {colorUseRule=UseColor_Specified, specifiedColor=<rgba 1 1 1 1>}
       uVRegion {bIsValid=false, max=<0 0>, min=<0 0>}
      }
      verticalBackgroundImage {
       bIsDynamicallyLoaded=false
       drawAs=NoDrawType
       imageType=NoImage
       mirroring=NoMirror
       resourceName=None
       resourceObject=null
       tiling=NoTile
       imageSize=<0 0>
       margin=<ltrb 0 0 0 0>
       outlineSettings {
        bUseBrushTransparency=false
        roundingType=HalfHeightRadius
        width=0
        color {colorUseRule=UseColor_Specified, specifiedColor=<rgba 0 0 0 0>}
        cornerRadii=<0 0 0 0>
       }
       tintColor {colorUseRule=UseColor_Specified, specifiedColor=<rgba 1 1 1 1>}
       uVRegion {bIsValid=false, max=<0 0>, min=<0 0>}
      }
      verticalBottomSlotImage {
       bIsDynamicallyLoaded=false
       drawAs=Image
       imageType=FullColor
       mirroring=NoMirror
       resourceName=../../../Engine/Content/Slate/Common/Scrollbar_Background_Vertical.png
       resourceObject=null
       tiling=NoTile
       imageSize=<8 8>
       margin=<ltrb 0 0 0 0>
       outlineSettings {
        bUseBrushTransparency=false
        roundingType=HalfHeightRadius
        width=0
        color {colorUseRule=UseColor_Specified, specifiedColor=<rgba 0 0 0 0>}
        cornerRadii=<0 0 0 0>
       }
       tintColor {colorUseRule=UseColor_Specified, specifiedColor=<rgba 1 1 1 1>}
       uVRegion {bIsValid=false, max=<0 0>, min=<0 0>}
      }
      verticalTopSlotImage {
       bIsDynamicallyLoaded=false
       drawAs=Image
       imageType=FullColor
       mirroring=NoMirror
       resourceName=../../../Engine/Content/Slate/Common/Scrollbar_Background_Vertical.png
       resourceObject=null
       tiling=NoTile
       imageSize=<8 8>
       margin=<ltrb 0 0 0 0>
       outlineSettings {
        bUseBrushTransparency=false
        roundingType=HalfHeightRadius
        width=0
        color {colorUseRule=UseColor_Specified, specifiedColor=<rgba 0 0 0 0>}
        cornerRadii=<0 0 0 0>
       }
       tintColor {colorUseRule=UseColor_Specified, specifiedColor=<rgba 1 1 1 1>}
       uVRegion {bIsValid=false, max=<0 0>, min=<0 0>}
      }
     }
     textStyle {
      overflowPolicy=Clip
      transformPolicy=None
      colorAndOpacity {colorUseRule=UseColor_Foreground, specifiedColor=<rgba 1 0 1 1>}
      font {
       bForceMonospaced=false
       bMaterialIsStencil=false
       fontMaterial=null
       fontObject=&BurbankSmall
       letterSpacing=0
       monospacedWidth=1
       size=24
       skewAmount=0
       typefaceFontName=Black
       outlineSettings {
        bApplyOutlineToDropShadows=false
        bMiteredCorners=false
        bSeparateFillAlpha=false
        outlineMaterial=null
        outlineSize=0
        outlineColor=<rgba 0 0 0 1>
       }
      }
      highlightColor {colorUseRule=UseColor_Specified, specifiedColor=<rgba 0.02 0.3 0 1>}
      highlightShape {
       bIsDynamicallyLoaded=false
       drawAs=Box
       imageType=FullColor
       mirroring=NoMirror
       resourceName=../../../Engine/Content/Slate/Common/TextBlockHighlightShape.png
       resourceObject=null
       tiling=NoTile
       imageSize=<0 0>
       margin=<ltrb 0.375 0.375 0.375 0.375>
       outlineSettings {
        bUseBrushTransparency=false
        roundingType=HalfHeightRadius
        width=0
        color {colorUseRule=UseColor_Specified, specifiedColor=<rgba 0 0 0 0>}
        cornerRadii=<0 0 0 0>
       }
       tintColor {colorUseRule=UseColor_Specified, specifiedColor=<rgba 1 1 1 1>}
       uVRegion {bIsValid=false, max=<0 0>, min=<0 0>}
      }
      selectedBackgroundColor {colorUseRule=UseColor_Foreground, specifiedColor=<rgba 1 0 1 1>}
      shadowColorAndOpacity=<rgba 0 0 0 1>
      shadowOffset=<0 0>
      strikeBrush {
       bIsDynamicallyLoaded=false
       drawAs=Image
       imageType=NoImage
       mirroring=NoMirror
       resourceName=None
       resourceObject=null
       tiling=NoTile
       imageSize=<32 32>
       margin=<ltrb 0 0 0 0>
       outlineSettings {
        bUseBrushTransparency=false
        roundingType=HalfHeightRadius
        width=0
        color {colorUseRule=UseColor_Specified, specifiedColor=<rgba 0 0 0 0>}
        cornerRadii=<0 0 0 0>
       }
       tintColor {colorUseRule=UseColor_Specified, specifiedColor=<rgba 1 1 1 1>}
       uVRegion {bIsValid=false, max=<0 0>, min=<0 0>}
      }
      underlineBrush {
       bIsDynamicallyLoaded=false
       drawAs=Image
       imageType=NoImage
       mirroring=NoMirror
       resourceName=None
       resourceObject=null
       tiling=NoTile
       imageSize=<32 32>
       margin=<ltrb 0 0 0 0>
       outlineSettings {
        bUseBrushTransparency=false
        roundingType=HalfHeightRadius
        width=0
        color {colorUseRule=UseColor_Specified, specifiedColor=<rgba 0 0 0 0>}
        cornerRadii=<0 0 0 0>
       }
       tintColor {colorUseRule=UseColor_Specified, specifiedColor=<rgba 1 1 1 1>}
       uVRegion {bIsValid=false, max=<0 0>, min=<0 0>}
      }
     }
     vScrollBarPadding=<ltrb 0 0 0 0>
    }
   }
   slot {class=SizeBoxSlot}
  }
  Text_CharCountTitle {
   widgetClassPath=&CommonTextBlock
   bInherited=true
   parent=VerticalBox_5
   widget=Text_CharCountTitle
   properties {
    style=/FortUILibrary/Styles/Typography/Foundation/TextStyle-Base-S-B-Gray.TextStyle-Base-S-B-Gray_C
    text=0 / 250
    colorAndOpacity {colorUseRule=UseColor_Specified, specifiedColor=<rgba 0.0976 0.0976 0.0976 1>}
    font {
     bForceMonospaced=false
     bMaterialIsStencil=false
     fontMaterial=null
     fontObject=&NotoSans
     letterSpacing=0
     monospacedWidth=1
     size=18
     skewAmount=0
     typefaceFontName=NotoSans-Bold
     outlineSettings {
      bApplyOutlineToDropShadows=false
      bMiteredCorners=false
      bSeparateFillAlpha=false
      outlineMaterial=null
      outlineSize=0
      outlineColor=<rgba 0 0 0 1>
     }
    }
    shadowOffset=<0 0>
   }
   slot {class=VerticalBoxSlot, properties {horizontalAlignment=HAlign_Right, padding=<ltrb 0 0 10 0>}}
  }
  Text_ErrorInNameField {
   widgetClassPath=&CommonTextBlock
   bInherited=true
   parent=VerticalBox_2
   widget=Text_ErrorInNameField
   properties {
    justification=Center
    text=Required Field
    colorAndOpacity {colorUseRule=UseColor_Specified, specifiedColor=<rgba 0.9301 0.0742 0.107 1>}
    font {
     bForceMonospaced=false
     bMaterialIsStencil=false
     fontMaterial=null
     fontObject=&BurbankSmall
     letterSpacing=0
     monospacedWidth=1
     size=24
     skewAmount=0
     typefaceFontName=Black
     outlineSettings {
      bApplyOutlineToDropShadows=false
      bMiteredCorners=false
      bSeparateFillAlpha=false
      outlineMaterial=null
      outlineSize=0
      outlineColor=<rgba 0 0 0 1>
     }
    }
    shadowOffset=<0 0>
   }
   slot {class=VerticalBoxSlot, properties {horizontalAlignment=HAlign_Center, padding=<ltrb 0 -25 0 0>}}
  }
  HorizontalBox_1 {
   widgetClassPath=&HorizontalBox
   parent=VB_PublishEntry
   widget=HorizontalBox_1
   slot {class=VerticalBoxSlot, properties {horizontalAlignment=HAlign_Center, verticalAlignment=VAlign_Top, padding=<ltrb 0 20 0 0>}}
  }
  CommonTextBlock_0 {
   widgetClassPath=&CommonTextBlock
   parent=HorizontalBox_1
   widget=CommonTextBlock_0
   properties {
    justification=Right
    text=DESCRIPTION:
    font {
     bForceMonospaced=false
     bMaterialIsStencil=false
     fontMaterial=null
     fontObject=&BurbankSmall
     letterSpacing=0
     monospacedWidth=1
     size=36
     skewAmount=0
     typefaceFontName=Black
     outlineSettings {
      bApplyOutlineToDropShadows=false
      bMiteredCorners=false
      bSeparateFillAlpha=false
      outlineMaterial=null
      outlineSize=0
      outlineColor=<rgba 0 0 0 1>
     }
    }
    shadowOffset=<0 0>
   }
   slot {class=HorizontalBoxSlot, properties {verticalAlignment=VAlign_Top, padding=<ltrb 0 15 20 0>, size {sizeRule=Fill, value=1}}}
  }
  VerticalBox_3 {
   widgetClassPath=&VerticalBox
   parent=HorizontalBox_1
   widget=VerticalBox_3
   slot {class=HorizontalBoxSlot, properties {padding=<ltrb 0 15 0 0>}}
  }
  SizeBox_1 {
   widgetClassPath=&SizeBox
   parent=VerticalBox_3
   widget=SizeBox_1
   properties {bOverride_WidthOverride=true, heightOverride=1000, widthOverride=1000}
   slot {class=VerticalBoxSlot}
  }
  EditText_Description {
   widgetClassPath=&EditableTextBox
   bInherited=true
   bIsVariable=true
   parent=SizeBox_1
   widget=EditText_Description
   properties {
    hintText="Tell everyone about your Island! "
    justification=Center
    navigation=&self:WidgetTree.EditText_Description.WidgetNavigation_0
    widgetStyle {
     backgroundColor {colorUseRule=UseColor_Specified, specifiedColor=<rgba 1 1 1 1>}
     backgroundImageFocused {
      bIsDynamicallyLoaded=false
      drawAs=Box
      imageType=FullColor
      mirroring=NoMirror
      resourceName=../../../Engine/Content/Slate/Common/TextBox_Hovered.png
      resourceObject=null
      tiling=NoTile
      imageSize=<0 0>
      margin=<ltrb 0.25 0.25 0.25 0.25>
      outlineSettings {
       bUseBrushTransparency=false
       roundingType=HalfHeightRadius
       width=0
       color {colorUseRule=UseColor_Specified, specifiedColor=<rgba 0 0 0 0>}
       cornerRadii=<0 0 0 0>
      }
      tintColor {colorUseRule=UseColor_Specified, specifiedColor=<rgba 1 1 1 1>}
      uVRegion {bIsValid=false, max=<0 0>, min=<0 0>}
     }
     backgroundImageHovered {
      bIsDynamicallyLoaded=false
      drawAs=Box
      imageType=FullColor
      mirroring=NoMirror
      resourceName=../../../Engine/Content/Slate/Common/TextBox_Hovered.png
      resourceObject=null
      tiling=NoTile
      imageSize=<0 0>
      margin=<ltrb 0.25 0.25 0.25 0.25>
      outlineSettings {
       bUseBrushTransparency=false
       roundingType=HalfHeightRadius
       width=0
       color {colorUseRule=UseColor_Specified, specifiedColor=<rgba 0 0 0 0>}
       cornerRadii=<0 0 0 0>
      }
      tintColor {colorUseRule=UseColor_Specified, specifiedColor=<rgba 1 1 1 1>}
      uVRegion {bIsValid=false, max=<0 0>, min=<0 0>}
     }
     backgroundImageNormal {
      bIsDynamicallyLoaded=false
      drawAs=Box
      imageType=FullColor
      mirroring=NoMirror
      resourceName=../../../Engine/Content/Slate/Common/TextBox.png
      resourceObject=null
      tiling=NoTile
      imageSize=<0 0>
      margin=<ltrb 0.25 0.25 0.25 0.25>
      outlineSettings {
       bUseBrushTransparency=false
       roundingType=HalfHeightRadius
       width=0
       color {colorUseRule=UseColor_Specified, specifiedColor=<rgba 0 0 0 0>}
       cornerRadii=<0 0 0 0>
      }
      tintColor {colorUseRule=UseColor_Specified, specifiedColor=<rgba 1 1 1 1>}
      uVRegion {bIsValid=false, max=<0 0>, min=<0 0>}
     }
     backgroundImageReadOnly {
      bIsDynamicallyLoaded=false
      drawAs=Box
      imageType=FullColor
      mirroring=NoMirror
      resourceName=../../../Engine/Content/Slate/Common/TextBox_ReadOnly.png
      resourceObject=null
      tiling=NoTile
      imageSize=<0 0>
      margin=<ltrb 0.25 0.25 0.25 0.25>
      outlineSettings {
       bUseBrushTransparency=false
       roundingType=HalfHeightRadius
       width=0
       color {colorUseRule=UseColor_Specified, specifiedColor=<rgba 0 0 0 0>}
       cornerRadii=<0 0 0 0>
      }
      tintColor {colorUseRule=UseColor_Specified, specifiedColor=<rgba 1 1 1 1>}
      uVRegion {bIsValid=false, max=<0 0>, min=<0 0>}
     }
     focusedForegroundColor {colorUseRule=UseColor_Foreground, specifiedColor=<rgba 1 0 1 1>}
     foregroundColor {colorUseRule=UseColor_Specified, specifiedColor=<rgba 0 0 0 1>}
     hScrollBarPadding=<ltrb 0 0 0 0>
     padding=<ltrb 6 10 6 10>
     readOnlyForegroundColor {colorUseRule=UseColor_Foreground, specifiedColor=<rgba 1 0 1 1>}
     scrollBarStyle {
      thickness=16
      draggedThumbImage {
       bIsDynamicallyLoaded=false
       drawAs=Box
       imageType=FullColor
       mirroring=NoMirror
       resourceName=../../../Engine/Content/Slate/Common/Scrollbar_Thumb.png
       resourceObject=null
       tiling=NoTile
       imageSize=<0 0>
       margin=<ltrb 0.25 0.25 0.25 0.25>
       outlineSettings {
        bUseBrushTransparency=false
        roundingType=HalfHeightRadius
        width=0
        color {colorUseRule=UseColor_Specified, specifiedColor=<rgba 0 0 0 0>}
        cornerRadii=<0 0 0 0>
       }
       tintColor {colorUseRule=UseColor_Specified, specifiedColor=<rgba 1 1 1 1>}
       uVRegion {bIsValid=false, max=<0 0>, min=<0 0>}
      }
      horizontalBackgroundImage {
       bIsDynamicallyLoaded=false
       drawAs=NoDrawType
       imageType=NoImage
       mirroring=NoMirror
       resourceName=None
       resourceObject=null
       tiling=NoTile
       imageSize=<0 0>
       margin=<ltrb 0 0 0 0>
       outlineSettings {
        bUseBrushTransparency=false
        roundingType=HalfHeightRadius
        width=0
        color {colorUseRule=UseColor_Specified, specifiedColor=<rgba 0 0 0 0>}
        cornerRadii=<0 0 0 0>
       }
       tintColor {colorUseRule=UseColor_Specified, specifiedColor=<rgba 1 1 1 1>}
       uVRegion {bIsValid=false, max=<0 0>, min=<0 0>}
      }
      horizontalBottomSlotImage {
       bIsDynamicallyLoaded=false
       drawAs=Image
       imageType=FullColor
       mirroring=NoMirror
       resourceName=../../../Engine/Content/Slate/Common/Scrollbar_Background_Horizontal.png
       resourceObject=null
       tiling=NoTile
       imageSize=<8 8>
       margin=<ltrb 0 0 0 0>
       outlineSettings {
        bUseBrushTransparency=false
        roundingType=HalfHeightRadius
        width=0
        color {colorUseRule=UseColor_Specified, specifiedColor=<rgba 0 0 0 0>}
        cornerRadii=<0 0 0 0>
       }
       tintColor {colorUseRule=UseColor_Specified, specifiedColor=<rgba 1 1 1 1>}
       uVRegion {bIsValid=false, max=<0 0>, min=<0 0>}
      }
      horizontalTopSlotImage {
       bIsDynamicallyLoaded=false
       drawAs=Image
       imageType=FullColor
       mirroring=NoMirror
       resourceName=../../../Engine/Content/Slate/Common/Scrollbar_Background_Horizontal.png
       resourceObject=null
       tiling=NoTile
       imageSize=<8 8>
       margin=<ltrb 0 0 0 0>
       outlineSettings {
        bUseBrushTransparency=false
        roundingType=HalfHeightRadius
        width=0
        color {colorUseRule=UseColor_Specified, specifiedColor=<rgba 0 0 0 0>}
        cornerRadii=<0 0 0 0>
       }
       tintColor {colorUseRule=UseColor_Specified, specifiedColor=<rgba 1 1 1 1>}
       uVRegion {bIsValid=false, max=<0 0>, min=<0 0>}
      }
      hoveredThumbImage {
       bIsDynamicallyLoaded=false
       drawAs=Box
       imageType=FullColor
       mirroring=NoMirror
       resourceName=../../../Engine/Content/Slate/Common/Scrollbar_Thumb.png
       resourceObject=null
       tiling=NoTile
       imageSize=<0 0>
       margin=<ltrb 0.25 0.25 0.25 0.25>
       outlineSettings {
        bUseBrushTransparency=false
        roundingType=HalfHeightRadius
        width=0
        color {colorUseRule=UseColor_Specified, specifiedColor=<rgba 0 0 0 0>}
        cornerRadii=<0 0 0 0>
       }
       tintColor {colorUseRule=UseColor_Specified, specifiedColor=<rgba 1 1 1 1>}
       uVRegion {bIsValid=false, max=<0 0>, min=<0 0>}
      }
      normalThumbImage {
       bIsDynamicallyLoaded=false
       drawAs=Box
       imageType=FullColor
       mirroring=NoMirror
       resourceName=../../../Engine/Content/Slate/Common/Scrollbar_Thumb.png
       resourceObject=null
       tiling=NoTile
       imageSize=<0 0>
       margin=<ltrb 0.25 0.25 0.25 0.25>
       outlineSettings {
        bUseBrushTransparency=false
        roundingType=HalfHeightRadius
        width=0
        color {colorUseRule=UseColor_Specified, specifiedColor=<rgba 0 0 0 0>}
        cornerRadii=<0 0 0 0>
       }
       tintColor {colorUseRule=UseColor_Specified, specifiedColor=<rgba 1 1 1 1>}
       uVRegion {bIsValid=false, max=<0 0>, min=<0 0>}
      }
      verticalBackgroundImage {
       bIsDynamicallyLoaded=false
       drawAs=NoDrawType
       imageType=NoImage
       mirroring=NoMirror
       resourceName=None
       resourceObject=null
       tiling=NoTile
       imageSize=<0 0>
       margin=<ltrb 0 0 0 0>
       outlineSettings {
        bUseBrushTransparency=false
        roundingType=HalfHeightRadius
        width=0
        color {colorUseRule=UseColor_Specified, specifiedColor=<rgba 0 0 0 0>}
        cornerRadii=<0 0 0 0>
       }
       tintColor {colorUseRule=UseColor_Specified, specifiedColor=<rgba 1 1 1 1>}
       uVRegion {bIsValid=false, max=<0 0>, min=<0 0>}
      }
      verticalBottomSlotImage {
       bIsDynamicallyLoaded=false
       drawAs=Image
       imageType=FullColor
       mirroring=NoMirror
       resourceName=../../../Engine/Content/Slate/Common/Scrollbar_Background_Vertical.png
       resourceObject=null
       tiling=NoTile
       imageSize=<8 8>
       margin=<ltrb 0 0 0 0>
       outlineSettings {
        bUseBrushTransparency=false
        roundingType=HalfHeightRadius
        width=0
        color {colorUseRule=UseColor_Specified, specifiedColor=<rgba 0 0 0 0>}
        cornerRadii=<0 0 0 0>
       }
       tintColor {colorUseRule=UseColor_Specified, specifiedColor=<rgba 1 1 1 1>}
       uVRegion {bIsValid=false, max=<0 0>, min=<0 0>}
      }
      verticalTopSlotImage {
       bIsDynamicallyLoaded=false
       drawAs=Image
       imageType=FullColor
       mirroring=NoMirror
       resourceName=../../../Engine/Content/Slate/Common/Scrollbar_Background_Vertical.png
       resourceObject=null
       tiling=NoTile
       imageSize=<8 8>
       margin=<ltrb 0 0 0 0>
       outlineSettings {
        bUseBrushTransparency=false
        roundingType=HalfHeightRadius
        width=0
        color {colorUseRule=UseColor_Specified, specifiedColor=<rgba 0 0 0 0>}
        cornerRadii=<0 0 0 0>
       }
       tintColor {colorUseRule=UseColor_Specified, specifiedColor=<rgba 1 1 1 1>}
       uVRegion {bIsValid=false, max=<0 0>, min=<0 0>}
      }
     }
     textStyle {
      overflowPolicy=Clip
      transformPolicy=None
      colorAndOpacity {colorUseRule=UseColor_Foreground, specifiedColor=<rgba 1 0 1 1>}
      font {
       bForceMonospaced=false
       bMaterialIsStencil=false
       fontMaterial=null
       fontObject=&BurbankSmall
       letterSpacing=0
       monospacedWidth=1
       size=24
       skewAmount=0
       typefaceFontName=Black
       outlineSettings {
        bApplyOutlineToDropShadows=false
        bMiteredCorners=false
        bSeparateFillAlpha=false
        outlineMaterial=null
        outlineSize=0
        outlineColor=<rgba 0 0 0 1>
       }
      }
      highlightColor {colorUseRule=UseColor_Specified, specifiedColor=<rgba 0.02 0.3 0 1>}
      highlightShape {
       bIsDynamicallyLoaded=false
       drawAs=Box
       imageType=FullColor
       mirroring=NoMirror
       resourceName=../../../Engine/Content/Slate/Common/TextBlockHighlightShape.png
       resourceObject=null
       tiling=NoTile
       imageSize=<0 0>
       margin=<ltrb 0.375 0.375 0.375 0.375>
       outlineSettings {
        bUseBrushTransparency=false
        roundingType=HalfHeightRadius
        width=0
        color {colorUseRule=UseColor_Specified, specifiedColor=<rgba 0 0 0 0>}
        cornerRadii=<0 0 0 0>
       }
       tintColor {colorUseRule=UseColor_Specified, specifiedColor=<rgba 1 1 1 1>}
       uVRegion {bIsValid=false, max=<0 0>, min=<0 0>}
      }
      selectedBackgroundColor {colorUseRule=UseColor_Foreground, specifiedColor=<rgba 1 0 1 1>}
      shadowColorAndOpacity=<rgba 0 0 0 1>
      shadowOffset=<0 0>
      strikeBrush {
       bIsDynamicallyLoaded=false
       drawAs=Image
       imageType=NoImage
       mirroring=NoMirror
       resourceName=None
       resourceObject=null
       tiling=NoTile
       imageSize=<32 32>
       margin=<ltrb 0 0 0 0>
       outlineSettings {
        bUseBrushTransparency=false
        roundingType=HalfHeightRadius
        width=0
        color {colorUseRule=UseColor_Specified, specifiedColor=<rgba 0 0 0 0>}
        cornerRadii=<0 0 0 0>
       }
       tintColor {colorUseRule=UseColor_Specified, specifiedColor=<rgba 1 1 1 1>}
       uVRegion {bIsValid=false, max=<0 0>, min=<0 0>}
      }
      underlineBrush {
       bIsDynamicallyLoaded=false
       drawAs=Image
       imageType=NoImage
       mirroring=NoMirror
       resourceName=None
       resourceObject=null
       tiling=NoTile
       imageSize=<32 32>
       margin=<ltrb 0 0 0 0>
       outlineSettings {
        bUseBrushTransparency=false
        roundingType=HalfHeightRadius
        width=0
        color {colorUseRule=UseColor_Specified, specifiedColor=<rgba 0 0 0 0>}
        cornerRadii=<0 0 0 0>
       }
       tintColor {colorUseRule=UseColor_Specified, specifiedColor=<rgba 1 1 1 1>}
       uVRegion {bIsValid=false, max=<0 0>, min=<0 0>}
      }
     }
     vScrollBarPadding=<ltrb 0 0 0 0>
    }
   }
   slot {class=SizeBoxSlot}
  }
  Text_CharCountDesc {
   widgetClassPath=&CommonTextBlock
   bInherited=true
   parent=VerticalBox_3
   widget=Text_CharCountDesc
   properties {
    style=&TextStyle-Base-S-B-Gray_C
    text=0 / 250
    colorAndOpacity {colorUseRule=UseColor_Specified, specifiedColor=<rgba 0.0976 0.0976 0.0976 1>}
    font {
     bForceMonospaced=false
     bMaterialIsStencil=false
     fontMaterial=null
     fontObject=&NotoSans
     letterSpacing=0
     monospacedWidth=1
     size=18
     skewAmount=0
     typefaceFontName=NotoSans-Bold
     outlineSettings {
      bApplyOutlineToDropShadows=false
      bMiteredCorners=false
      bSeparateFillAlpha=false
      outlineMaterial=null
      outlineSize=0
      outlineColor=<rgba 0 0 0 1>
     }
    }
    shadowOffset=<0 0>
   }
   slot {class=VerticalBoxSlot, properties {horizontalAlignment=HAlign_Right, padding=<ltrb 0 0 10 0>}}
  }
  Text_ErrorInDescriptionField {
   widgetClassPath=&CommonTextBlock
   bInherited=true
   parent=VerticalBox_3
   widget=Text_ErrorInDescriptionField
   properties {
    justification=Center
    text=Required Field
    colorAndOpacity {colorUseRule=UseColor_Specified, specifiedColor=<rgba 0.9301 0.0742 0.107 1>}
    font {
     bForceMonospaced=false
     bMaterialIsStencil=false
     fontMaterial=null
     fontObject=&BurbankSmall
     letterSpacing=0
     monospacedWidth=1
     size=24
     skewAmount=0
     typefaceFontName=Black
     outlineSettings {
      bApplyOutlineToDropShadows=false
      bMiteredCorners=false
      bSeparateFillAlpha=false
      outlineMaterial=null
      outlineSize=0
      outlineColor=<rgba 0 0 0 1>
     }
    }
    shadowOffset=<0 0>
   }
   slot {class=VerticalBoxSlot, properties {horizontalAlignment=HAlign_Center, padding=<ltrb 0 -25 0 0>}}
  }
  HBox_ClearAllPlayersData {
   widgetClassPath=&HorizontalBox
   bIsVariable=true
   parent=VB_PublishEntry
   widget=HBox_ClearAllPlayersData
   slot {class=VerticalBoxSlot, properties {horizontalAlignment=HAlign_Center, verticalAlignment=VAlign_Top}}
  }
  HorizontalBox {
   widgetClassPath=&HorizontalBox
   parent=HBox_ClearAllPlayersData
   widget=HorizontalBox
   slot {class=HorizontalBoxSlot, properties {verticalAlignment=VAlign_Top}}
  }
  SizeBox {
   widgetClassPath=&SizeBox
   parent=HorizontalBox
   widget=SizeBox
   properties {bOverride_HeightOverride=true, bOverride_WidthOverride=true, heightOverride=75, widthOverride=75}
   slot {class=HorizontalBoxSlot, properties {horizontalAlignment=HAlign_Right, verticalAlignment=VAlign_Top}}
  }
  CheckBox_ClearAllPlayersData {
   widgetClassPath=/Script/UMG.CheckBox
   bInherited=true
   parent=SizeBox
   widget=CheckBox_ClearAllPlayersData
   properties {
    horizontalAlignment=HAlign_Center
    navigation=&self:WidgetTree.CheckBox_ClearAllPlayersData.WidgetNavigation_0
    widgetStyle {
     checkBoxType=ToggleButton
     backgroundHoveredImage {
      bIsDynamicallyLoaded=false
      drawAs=NoDrawType
      imageType=NoImage
      mirroring=NoMirror
      resourceName=None
      resourceObject=null
      tiling=NoTile
      imageSize=<0 0>
      margin=<ltrb 0 0 0 0>
      outlineSettings {
       bUseBrushTransparency=false
       roundingType=HalfHeightRadius
       width=0
       color {colorUseRule=UseColor_Specified, specifiedColor=<rgba 0 0 0 0>}
       cornerRadii=<0 0 0 0>
      }
      tintColor {colorUseRule=UseColor_Specified, specifiedColor=<rgba 1 1 1 1>}
      uVRegion {bIsValid=false, max=<0 0>, min=<0 0>}
     }
     backgroundImage {
      bIsDynamicallyLoaded=false
      drawAs=NoDrawType
      imageType=NoImage
      mirroring=NoMirror
      resourceName=None
      resourceObject=null
      tiling=NoTile
      imageSize=<0 0>
      margin=<ltrb 0 0 0 0>
      outlineSettings {
       bUseBrushTransparency=false
       roundingType=HalfHeightRadius
       width=0
       color {colorUseRule=UseColor_Specified, specifiedColor=<rgba 0 0 0 0>}
       cornerRadii=<0 0 0 0>
      }
      tintColor {colorUseRule=UseColor_Specified, specifiedColor=<rgba 1 1 1 1>}
      uVRegion {bIsValid=false, max=<0 0>, min=<0 0>}
     }
     backgroundPressedImage {
      bIsDynamicallyLoaded=false
      drawAs=NoDrawType
      imageType=NoImage
      mirroring=NoMirror
      resourceName=None
      resourceObject=null
      tiling=NoTile
      imageSize=<0 0>
      margin=<ltrb 0 0 0 0>
      outlineSettings {
       bUseBrushTransparency=false
       roundingType=HalfHeightRadius
       width=0
       color {colorUseRule=UseColor_Specified, specifiedColor=<rgba 0 0 0 0>}
       cornerRadii=<0 0 0 0>
      }
      tintColor {colorUseRule=UseColor_Specified, specifiedColor=<rgba 1 1 1 1>}
      uVRegion {bIsValid=false, max=<0 0>, min=<0 0>}
     }
     borderBackgroundColor {colorUseRule=UseColor_Specified, specifiedColor=<rgba 1 1 1 1>}
     checkedForeground {colorUseRule=UseColor_Foreground, specifiedColor=<rgba 1 0 1 1>}
     checkedHoveredForeground {colorUseRule=UseColor_Foreground, specifiedColor=<rgba 1 0 1 1>}
     checkedHoveredImage {
      bIsDynamicallyLoaded=false
      drawAs=Image
      imageType=FullColor
      mirroring=NoMirror
      resourceName=../../../Engine/Content/Slate/Common/CheckBox_Checked_Hovered.png
      resourceObject=/Game/UI/Foundation/Textures/Icons/Checks/T-Icon-CheckedBox-76
      tiling=NoTile
      imageSize=<76 76>
      margin=<ltrb 0 0 0 0>
      outlineSettings {
       bUseBrushTransparency=false
       roundingType=HalfHeightRadius
       width=0
       color {colorUseRule=UseColor_Specified, specifiedColor=<rgba 0 0 0 0>}
       cornerRadii=<0 0 0 0>
      }
      tintColor {colorUseRule=UseColor_Specified, specifiedColor=<rgba 0.6038 0.6445 0.6724 1>}
      uVRegion {bIsValid=false, max=<0 0>, min=<0 0>}
     }
     checkedImage {
      bIsDynamicallyLoaded=false
      drawAs=Image
      imageType=FullColor
      mirroring=NoMirror
      resourceName=../../../Engine/Content/Slate/Common/CheckBox_Checked_Hovered.png
      resourceObject=&T-Icon-CheckedBox-76
      tiling=NoTile
      imageSize=<76 76>
      margin=<ltrb 0 0 0 0>
      outlineSettings {
       bUseBrushTransparency=false
       roundingType=HalfHeightRadius
       width=0
       color {colorUseRule=UseColor_Specified, specifiedColor=<rgba 0 0 0 0>}
       cornerRadii=<0 0 0 0>
      }
      tintColor {colorUseRule=UseColor_Specified, specifiedColor=<rgba 0.2462 0.9647 0.3278 1>}
      uVRegion {bIsValid=false, max=<0 0>, min=<0 0>}
     }
     checkedPressedForeground {colorUseRule=UseColor_Foreground, specifiedColor=<rgba 1 0 1 1>}
     checkedPressedImage {
      bIsDynamicallyLoaded=false
      drawAs=Image
      imageType=FullColor
      mirroring=NoMirror
      resourceName=../../../Engine/Content/Slate/Common/CheckBox_Checked.png
      resourceObject=&T-Icon-CheckedBox-76
      tiling=NoTile
      imageSize=<76 76>
      margin=<ltrb 0 0 0 0>
      outlineSettings {
       bUseBrushTransparency=false
       roundingType=HalfHeightRadius
       width=0
       color {colorUseRule=UseColor_Specified, specifiedColor=<rgba 0 0 0 0>}
       cornerRadii=<0 0 0 0>
      }
      tintColor {colorUseRule=UseColor_Specified, specifiedColor=<rgba 0.6038 0.6445 0.6724 1>}
      uVRegion {bIsValid=false, max=<0 0>, min=<0 0>}
     }
     checkedSlateSound {resourceObject=null}
     foregroundColor {colorUseRule=UseColor_Foreground, specifiedColor=<rgba 1 0 1 1>}
     hoveredForeground {colorUseRule=UseColor_Foreground, specifiedColor=<rgba 1 0 1 1>}
     hoveredSlateSound {resourceObject=null}
     padding=<ltrb 2 0 0 0>
     pressedForeground {colorUseRule=UseColor_Foreground, specifiedColor=<rgba 1 0 1 1>}
     uncheckedHoveredImage {
      bIsDynamicallyLoaded=false
      drawAs=Image
      imageType=FullColor
      mirroring=NoMirror
      resourceName=../../../Engine/Content/Slate/Common/CheckBox.png
      resourceObject=/Game/UI/Foundation/Textures/Icons/Checks/T-Icon-Confirmation_Empty-76
      tiling=NoTile
      imageSize=<76 76>
      margin=<ltrb 0 0 0 0>
      outlineSettings {
       bUseBrushTransparency=false
       roundingType=HalfHeightRadius
       width=0
       color {colorUseRule=UseColor_Specified, specifiedColor=<rgba 0 0 0 0>}
       cornerRadii=<0 0 0 0>
      }
      tintColor {colorUseRule=UseColor_Specified, specifiedColor=<rgba 0.6038 0.6445 0.6724 1>}
      uVRegion {bIsValid=false, max=<0 0>, min=<0 0>}
     }
     uncheckedImage {
      bIsDynamicallyLoaded=false
      drawAs=Image
      imageType=FullColor
      mirroring=NoMirror
      resourceName=../../../Engine/Content/Slate/Common/CheckBox.png
      resourceObject=&T-Icon-Confirmation_Empty-76
      tiling=NoTile
      imageSize=<76 76>
      margin=<ltrb 0 0 0 0>
      outlineSettings {
       bUseBrushTransparency=false
       roundingType=HalfHeightRadius
       width=0
       color {colorUseRule=UseColor_Specified, specifiedColor=<rgba 0 0 0 0>}
       cornerRadii=<0 0 0 0>
      }
      tintColor {colorUseRule=UseColor_Specified, specifiedColor=<rgba 1 1 1 1>}
      uVRegion {bIsValid=false, max=<0 0>, min=<0 0>}
     }
     uncheckedPressedImage {
      bIsDynamicallyLoaded=false
      drawAs=Image
      imageType=FullColor
      mirroring=NoMirror
      resourceName=../../../Engine/Content/Slate/Common/CheckBox_Hovered.png
      resourceObject=&T-Icon-Confirmation_Empty-76
      tiling=NoTile
      imageSize=<76 76>
      margin=<ltrb 0 0 0 0>
      outlineSettings {
       bUseBrushTransparency=false
       roundingType=HalfHeightRadius
       width=0
       color {colorUseRule=UseColor_Specified, specifiedColor=<rgba 0 0 0 0>}
       cornerRadii=<0 0 0 0>
      }
      tintColor {colorUseRule=UseColor_Specified, specifiedColor=<rgba 0.6038 0.6445 0.6724 1>}
      uVRegion {bIsValid=false, max=<0 0>, min=<0 0>}
     }
     uncheckedSlateSound {resourceObject=null}
     undeterminedForeground {colorUseRule=UseColor_Foreground, specifiedColor=<rgba 1 0 1 1>}
     undeterminedHoveredImage {
      bIsDynamicallyLoaded=false
      drawAs=Image
      imageType=FullColor
      mirroring=NoMirror
      resourceName=../../../Engine/Content/Slate/Common/CheckBox_Undetermined_Hovered.png
      resourceObject=null
      tiling=NoTile
      imageSize=<32 32>
      margin=<ltrb 0 0 0 0>
      outlineSettings {
       bUseBrushTransparency=false
       roundingType=HalfHeightRadius
       width=0
       color {colorUseRule=UseColor_Specified, specifiedColor=<rgba 0 0 0 0>}
       cornerRadii=<0 0 0 0>
      }
      tintColor {colorUseRule=UseColor_Specified, specifiedColor=<rgba 1 1 1 1>}
      uVRegion {bIsValid=false, max=<0 0>, min=<0 0>}
     }
     undeterminedImage {
      bIsDynamicallyLoaded=false
      drawAs=Image
      imageType=FullColor
      mirroring=NoMirror
      resourceName=../../../Engine/Content/Slate/Common/CheckBox_Undetermined.png
      resourceObject=null
      tiling=NoTile
      imageSize=<32 32>
      margin=<ltrb 0 0 0 0>
      outlineSettings {
       bUseBrushTransparency=false
       roundingType=HalfHeightRadius
       width=0
       color {colorUseRule=UseColor_Specified, specifiedColor=<rgba 0 0 0 0>}
       cornerRadii=<0 0 0 0>
      }
      tintColor {colorUseRule=UseColor_Specified, specifiedColor=<rgba 1 1 1 1>}
      uVRegion {bIsValid=false, max=<0 0>, min=<0 0>}
     }
     undeterminedPressedImage {
      bIsDynamicallyLoaded=false
      drawAs=Image
      imageType=FullColor
      mirroring=NoMirror
      resourceName=../../../Engine/Content/Slate/Common/CheckBox_Undetermined_Hovered.png
      resourceObject=null
      tiling=NoTile
      imageSize=<32 32>
      margin=<ltrb 0 0 0 0>
      outlineSettings {
       bUseBrushTransparency=false
       roundingType=HalfHeightRadius
       width=0
       color {colorUseRule=UseColor_Specified, specifiedColor=<rgba 0 0 0 0>}
       cornerRadii=<0 0 0 0>
      }
      tintColor {colorUseRule=UseColor_Specified, specifiedColor=<rgba 0.5 0.5 0.5 1>}
      uVRegion {bIsValid=false, max=<0 0>, min=<0 0>}
     }
    }
   }
   slot {class=SizeBoxSlot}
  }
  Title_ClearAllPlayersData {
   widgetClassPath=&CommonTextBlock
   parent=HorizontalBox
   widget=Title_ClearAllPlayersData
   properties {
    justification=Right
    text=CLEAR ALL PLAYER'S SAVE DATA
    font {
     bForceMonospaced=false
     bMaterialIsStencil=false
     fontMaterial=null
     fontObject=&BurbankSmall
     letterSpacing=0
     monospacedWidth=1
     size=24
     skewAmount=0
     typefaceFontName=Black
     outlineSettings {
      bApplyOutlineToDropShadows=false
      bMiteredCorners=false
      bSeparateFillAlpha=false
      outlineMaterial=null
      outlineSize=0
      outlineColor=<rgba 0 0 0 1>
     }
    }
    shadowOffset=<0 0>
   }
   slot {class=HorizontalBoxSlot, properties {verticalAlignment=VAlign_Center, padding=<ltrb 10 8 0 0>}}
  }
  Text_ErrorInIslandBudget {
   widgetClassPath=&CommonTextBlock
   bInherited=true
   parent=VB_PublishEntry
   widget=Text_ErrorInIslandBudget
   properties {
    justification=Center
    text="Island is over budget!"
    visibility=Hidden
    colorAndOpacity {colorUseRule=UseColor_Specified, specifiedColor=<rgba 0.9301 0.0742 0.107 1>}
    font {
     bForceMonospaced=false
     bMaterialIsStencil=false
     fontMaterial=null
     fontObject=&BurbankSmall
     letterSpacing=0
     monospacedWidth=1
     size=24
     skewAmount=0
     typefaceFontName=Black
     outlineSettings {
      bApplyOutlineToDropShadows=false
      bMiteredCorners=false
      bSeparateFillAlpha=false
      outlineMaterial=null
      outlineSize=0
      outlineColor=<rgba 0 0 0 1>
     }
    }
    shadowOffset=<0 0>
   }
   slot {class=VerticalBoxSlot, properties {horizontalAlignment=HAlign_Center, padding=<ltrb 0 25 0 0>}}
  }
  Text_ErrorInRequiredTags {
   widgetClassPath=&CommonTextBlock
   bInherited=true
   parent=VB_PublishEntry
   widget=Text_ErrorInRequiredTags
   properties {
    justification=Center
    text="{Tag error}"
    colorAndOpacity {colorUseRule=UseColor_Specified, specifiedColor=<rgba 0.9301 0.0742 0.107 1>}
    font {
     bForceMonospaced=false
     bMaterialIsStencil=false
     fontMaterial=null
     fontObject=&BurbankSmall
     letterSpacing=0
     monospacedWidth=1
     size=24
     skewAmount=0
     typefaceFontName=Black
     outlineSettings {
      bApplyOutlineToDropShadows=false
      bMiteredCorners=false
      bSeparateFillAlpha=false
      outlineMaterial=null
      outlineSize=0
      outlineColor=<rgba 0 0 0 1>
     }
    }
    shadowOffset=<0 0>
   }
   slot {class=VerticalBoxSlot, properties {horizontalAlignment=HAlign_Center, padding=<ltrb 0 25 0 0>}}
  }
  HorizontalBox_0 {
   widgetClassPath=&HorizontalBox
   parent=VB_PublishEntry
   widget=HorizontalBox_0
   slot {
    class=VerticalBoxSlot
    properties {horizontalAlignment=HAlign_Center, verticalAlignment=VAlign_Center, padding=<ltrb 50 15 50 0>, size {sizeRule=Fill, value=1}}
   }
  }
  Image_Spinner {
   widgetClassPath=&Image
   bIsVariable=true
   parent=HorizontalBox_0
   widget=Image_Spinner
   properties {
    visibility=Hidden
    brush {
     bIsDynamicallyLoaded=false
     drawAs=Image
     imageType=NoImage
     mirroring=NoMirror
     resourceName=None
     resourceObject=/FortUILibrary/UIKit/Throbber/MI_UIKit_Monochrome_Throbber
     tiling=NoTile
     imageSize=<40 40>
     margin=<ltrb 0 0 0 0>
     outlineSettings {
      bUseBrushTransparency=false
      roundingType=HalfHeightRadius
      width=0
      color {colorUseRule=UseColor_Specified, specifiedColor=<rgba 0 0 0 0>}
      cornerRadii=<0 0 0 0>
     }
     tintColor {colorUseRule=UseColor_Specified, specifiedColor=<rgba 1 1 1 1>}
     uVRegion {bIsValid=false, max=<0 0>, min=<0 0>}
    }
   }
   slot {class=HorizontalBoxSlot, properties {verticalAlignment=VAlign_Center, padding=<ltrb 0 0 16 0>}}
  }
  Button_Submit {
   widgetClassPath=/FortniteCoreUI/Frontend/Showdown/PrimarySkewButton
   bInherited=true
   parent=HorizontalBox_0
   widget=Button_Submit
   properties {
    buttonNameText=PUBLISH
    inputPriority=100
    minHeight=1
    onButtonBaseClicked=(&self:WidgetTree.Button_Submit.HandleOnButtonClicked)
    onButtonBaseDoubleClicked=(&self:WidgetTree.Button_Submit.HandleOnButtonDoubleClicked)
    onButtonBaseHovered=(&self:WidgetTree.Button_Submit.HandleOnButtonHovered)
    onButtonBaseUnhovered=(&self:WidgetTree.Button_Submit.HandleOnButtonUnhovered)
    onSelectedChangedBase=(&self:WidgetTree.Button_Submit.HandleOnSelectedChanged)
    priority=100
    renderTransformPivot=<0.5 1>
    triggeringInputAction {dataTable=&InputActionDataTable, rowName=GenericAccept}
   }
   slot {class=HorizontalBoxSlot, properties {padding=<ltrb 0 0 66 0>}}
  }
  Button_Cancel {
   widgetClassPath=&IconTextButton
   bIsVariable=true
   parent=VB_PublishEntry
   widget=Button_Cancel
   properties {
    bIsFocusable=false
    buttonText=CANCEL
    contentAlignment=HAlign_Center
    minWidth=200
    navigation=&self:WidgetTree.Button_Cancel.WidgetNavigation_0
    onButtonBaseClicked=(&self:WidgetTree.Button_Cancel.HandleOnButtonClicked)
    onButtonBaseDoubleClicked=(&self:WidgetTree.Button_Cancel.HandleOnButtonDoubleClicked)
    onButtonBaseHovered=(&self:WidgetTree.Button_Cancel.HandleOnButtonHovered)
    onButtonBaseUnhovered=(&self:WidgetTree.Button_Cancel.HandleOnButtonUnhovered)
    onSelectedChangedBase=(&self:WidgetTree.Button_Cancel.HandleOnSelectedChanged)
    style=/FortUILibrary/Foundation/Buttons/ButtonStyles/Skew/ButtonStyle-Skew_Error.ButtonStyle-Skew_Error_C
    triggeringInputAction {dataTable=&InputActionDataTable, rowName=GenericBack}
   }
   slot {class=VerticalBoxSlot, properties {horizontalAlignment=HAlign_Right, verticalAlignment=VAlign_Bottom, padding=<ltrb 0 0 10 10>}}
  }
  VerticalBox_4 {widgetClassPath=&VerticalBox, parent=Switcher_Content, widget=VerticalBox_4, slot {class=WidgetSwitcherSlot}}
  Button_CopyCode {
   widgetClassPath=&IconTextButton
   bInherited=true
   bIsVariable=true
   parent=VerticalBox_4
   widget=Button_CopyCode
   properties {
    bIsFocusable=false
    buttonText=COPY ISLAND CODE
    contentAlignment=HAlign_Center
    inputPriority=100
    minWidth=365
    navigation=&self:WidgetTree.Button_CopyCode.WidgetNavigation_0
    onButtonBaseClicked=(&self:WidgetTree.Button_CopyCode.HandleOnButtonClicked)
    onButtonBaseDoubleClicked=(&self:WidgetTree.Button_CopyCode.HandleOnButtonDoubleClicked)
    onButtonBaseHovered=(&self:WidgetTree.Button_CopyCode.HandleOnButtonHovered)
    onButtonBaseUnhovered=(&self:WidgetTree.Button_CopyCode.HandleOnButtonUnhovered)
    onSelectedChangedBase=(&self:WidgetTree.Button_CopyCode.HandleOnSelectedChanged)
    priority=100
    style=&ButtonStyle-Skew_Error_C
    visibility=Collapsed
    triggeringInputAction {dataTable=&InputActionDataTable, rowName=GenericAccept}
   }
   slot {class=VerticalBoxSlot, properties {horizontalAlignment=HAlign_Center, verticalAlignment=VAlign_Bottom}}
  }
  Text_IslandPublishedSuccess {
   widgetClassPath=&CommonTextBlock
   bInherited=true
   bIsVariable=true
   parent=VerticalBox_4
   widget=Text_IslandPublishedSuccess
   properties {
    justification=Center
    text=YOUR ISLAND HAS BEEN PUBLISHED
    font {
     bForceMonospaced=false
     bMaterialIsStencil=false
     fontMaterial=null
     fontObject=&BurbankSmall
     letterSpacing=0
     monospacedWidth=1
     size=36
     skewAmount=0
     typefaceFontName=Black
     outlineSettings {
      bApplyOutlineToDropShadows=false
      bMiteredCorners=false
      bSeparateFillAlpha=false
      outlineMaterial=null
      outlineSize=0
      outlineColor=<rgba 0 0 0 1>
     }
    }
    shadowOffset=<0 0>
   }
   slot {class=VerticalBoxSlot, properties {horizontalAlignment=HAlign_Center, verticalAlignment=VAlign_Center, padding=<ltrb 30 30 30 30>}}
  }
  Button_Back {
   widgetClassPath=&IconTextButton
   bIsVariable=true
   parent=VerticalBox_4
   widget=Button_Back
   properties {
    bIsFocusable=false
    buttonText=BACK
    contentAlignment=HAlign_Center
    navigation=&self:WidgetTree.Button_Back.WidgetNavigation_0
    onButtonBaseClicked=(&self:WidgetTree.Button_Back.HandleOnButtonClicked)
    onButtonBaseDoubleClicked=(&self:WidgetTree.Button_Back.HandleOnButtonDoubleClicked)
    onButtonBaseHovered=(&self:WidgetTree.Button_Back.HandleOnButtonHovered)
    onButtonBaseUnhovered=(&self:WidgetTree.Button_Back.HandleOnButtonUnhovered)
    onSelectedChangedBase=(&self:WidgetTree.Button_Back.HandleOnSelectedChanged)
    style=&ButtonStyle-Skew_Error_C
    triggeringInputAction {dataTable=&InputActionDataTable, rowName=GenericBack}
   }
   slot {
    class=VerticalBoxSlot
    properties {horizontalAlignment=HAlign_Right, verticalAlignment=VAlign_Bottom, padding=<ltrb 0 0 10 10>, size {sizeRule=Fill, value=1}}
   }
  }
  CommonTextBlock_8 {
   widgetClassPath=&CommonTextBlock
   parent=VerticalBox_4
   widget=CommonTextBlock_8
   properties {
    justification=Center
    renderOpacity=0
    text="Island Code Copied to Clipboard!"
    visibility=Hidden
    colorAndOpacity {colorUseRule=UseColor_Specified, specifiedColor=<rgba 1 0.7622 0 1>}
    font {
     bForceMonospaced=false
     bMaterialIsStencil=false
     fontMaterial=null
     fontObject=&BurbankSmall
     letterSpacing=0
     monospacedWidth=1
     size=36
     skewAmount=0
     typefaceFontName=Black
     outlineSettings {
      bApplyOutlineToDropShadows=false
      bMiteredCorners=false
      bSeparateFillAlpha=false
      outlineMaterial=null
      outlineSize=0
      outlineColor=<rgba 0 0 0 1>
     }
    }
    renderTransform {angle=0, scale=<1 1>, shear=<0 0>, translation=<-46 -72>}
    shadowOffset=<0 0>
   }
   slot {class=VerticalBoxSlot, properties {horizontalAlignment=HAlign_Center, verticalAlignment=VAlign_Top, padding=<ltrb 0 0 0 -60>}}
  }
 ]
}
