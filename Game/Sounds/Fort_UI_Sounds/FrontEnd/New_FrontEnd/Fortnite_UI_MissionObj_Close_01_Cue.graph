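# UI sound cue for closing a mission objective in the front end, routed through the Fort_UI sound class.
# A mixer layers a quieter scroll-click wave with a mission objective pop-up wave, which passes through a Delay node left at its default timing.
# A concurrency override allows one instance at a time, stopping the oldest when retriggered.
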SoundCue /Game/Sounds/Fort_UI_Sounds/FrontEnd/New_FrontEnd/Fortnite_UI_MissionObj_Close_01_Cue
root=Mixer_0
nodes [
 {id=Delay_0, children [WavePlayer_3]}
 {id=Mixer_0, children [WavePlayer_2,Delay_0], properties {inputVolume [0.4,1]}}
 {id=WavePlayer_2, properties {soundWaveAssetPtr=&/UI_FN_ScrollClick02}}
 {id=WavePlayer_3, properties {soundWaveAssetPtr=&/UI_FN_Mission_Obj_PopUp_01}}
]
properties {
 bOverrideConcurrency=true
 maxDistance=2097152
 soundClassObject=&Fort_UI
 volumeMultiplier=1.75
 concurrencyOverrides {maxCount=1, resolutionRule=StopOldest}
}
