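# Creative modal that creates a private island code and points to the Creator Portal for publishing, built on FortCreativeCreateBuildModal.
# On activation, a game-state persistence-device check opens either the clear-save-data panel or the in-progress panel, skipping the switcher transition.
# Active-panel changes drive the primary button, clear-data input and intro animation.
# Parent-class events toggle device warnings, QR code visibility and copied-code feedback, while Reset restores defaults.
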WidgetBlueprint /FortniteCoreUI/InGame/Creative/WBP_Creative_CreateBuildModal
animations [
 ShowUp {
  binding_count=4
  length_seconds=0.45
  object=&self:ShowUp
  sequence {bindings [
    Waiting_VBox {
     id=#1
     possessed_class=VerticalBox
     tracks [
      {
       type=MovieSceneFloatTrack
       display_name=Render Opacity
       sections [{type=MovieSceneFloatSection, channels [None {type=MovieSceneScriptingFloatChannel, keys [[0,0],[2,0],[7,1]]}]}]
      }
      {
       type=MovieScene2DTransformTrack
       sections [{
         type=MovieScene2DTransformSection
         channels [
          Translation.X {type=MovieSceneScriptingFloatChannel}
          Translation.Y {type=MovieSceneScriptingFloatChannel, keys [[2,100],[7,0]]}
          Angle {type=MovieSceneScriptingFloatChannel}
          Scale.X {type=MovieSceneScriptingFloatChannel}
          Scale.Y {type=MovieSceneScriptingFloatChannel}
          Shear.X {type=MovieSceneScriptingFloatChannel}
          Shear.Y {type=MovieSceneScriptingFloatChannel}
         ]
        }]
      }
     ]
    }
    Button_PrimaryFunction {
     id=#2
     possessed_class=WBP_UIKit_Button_Regular_C
     tracks [{
       type=MovieScene2DTransformTrack
       sections [{
         type=MovieScene2DTransformSection
         channels [
          Translation.X {type=MovieSceneScriptingFloatChannel, const [0,1]}
          Translation.Y {type=MovieSceneScriptingFloatChannel, const [0,1]}
          Angle {type=MovieSceneScriptingFloatChannel}
          Scale.X {type=MovieSceneScriptingFloatChannel, keys [[0,0],[3,0],[5,1],[9,1]]}
          Scale.Y {type=MovieSceneScriptingFloatChannel, keys [[0,0],[3,0],[5,1],[9,1]]}
          Shear.X {type=MovieSceneScriptingFloatChannel}
          Shear.Y {type=MovieSceneScriptingFloatChannel}
         ]
        }]
      }]
    }
    Image_BottomLine {
     id=#3
     possessed_class=Image
     tracks [{
       type=MovieSceneWidgetMaterialTrack
       sections [{type=MovieSceneParameterSection, channels [Shape_Offset_Top {type=MovieSceneScriptingFloatChannel, keys [[0,-1.5],[6,-0.25]]}]}]
      }]
    }
    Containment {
     id=#4
     possessed_class=Image
     tracks [{
       type=MovieScene2DTransformTrack
       sections [{
         type=MovieScene2DTransformSection
         channels [
          Scale.X {type=MovieSceneScriptingFloatChannel}
          Scale.Y {type=MovieSceneScriptingFloatChannel, keys [[1,0],[6,1]]}
         ]
        }]
      }]
    }
   ]}
 }
 MnemonicCopied {
  binding_count=1
  length_seconds=0.35
  object=&self:MnemonicCopied
  sequence {bindings [Text_MnemonicCopied {
     id=#5
     possessed_class=CommonTextBlock
     tracks [
      {
       type=MovieSceneFloatTrack
       display_name=Render Opacity
       sections [{type=MovieSceneFloatSection, end_frame=7, start_frame=0, channels [None {type=MovieSceneScriptingFloatChannel, keys [[0,0],[4,1]]}]}]
      }
      {
       type=MovieScene2DTransformTrack
       sections [{
         type=MovieScene2DTransformSection
         end_frame=7
         start_frame=0
         channels [
          Scale.X {type=MovieSceneScriptingFloatChannel, keys [[0,0],[4,1.1],[6,1]]}
          Scale.Y {type=MovieSceneScriptingFloatChannel, keys [[0,0],[4,1.1],[6,1]]}
         ]
        }]
      }
     ]
    }]}
 }
]
graphs [
 EventGraph {nodes [
   n1 = "On Selected Changed (Button_ClearPersistentData)"!() -> (OutputDelegate, then, Button:Common Button Base Object Reference, Selected:bool) {is=title K2Node_ComponentBoundEvent}
   > SetIcon!(
    execute
    self:WBP UIKit Button Generic Base Object Reference=Get Button_ClearPersistentData() -> (Button_ClearPersistentData:WBP UIKit Button Generic Object Reference)
    Icon:Object Reference=Select(
     Option 0:Texture 2D Object Reference=/Game/UI/Foundation/Textures/Icons/Checks/T-Icon-Confirmation_Empty-76
     Option 1:Texture 2D Object Reference=/Game/UI/Foundation/Textures/Icons/Checks/T-Icon-CheckedBox-76
     Index=@n1.Selected
    ) -> (ReturnValue:Texture 2D Object Reference)
   ) -> (then)
   > SetClearPersistentData!(execute, self:Fort Creative Create Build Modal Object Reference, bInClearPersistentData:bool=@n1.Selected) -> (then)
   n2 = Event OnSwitcherTransitioningChanged!() -> (OutputDelegate, then, bIsTransitioning:bool) {is=title K2Node_Event}
   > SetIsInteractionEnabled!(
    execute
    self:Common Button Base Object Reference=Get Button_PrimaryFunction() -> (Button_PrimaryFunction:Common Button Base Object Reference)
    bInIsInteractionEnabled:bool=NOT Boolean(A=@n2.bIsTransitioning)
   ) -> (then)
   Event Construct()
   > SetVisibility(
    self=Get VBox_CreatorPortalUrl() -> (VBox_CreatorPortalUrl:Vertical Box Object Reference)
    InVisibility=GetVisibility!(self:Widget Object Reference=Get Text_CreatorPortalUrl() -> (Text_CreatorPortalUrl:Common Text Object Reference)) -> (ReturnValue:ESlateVisibility Enum=Visible)
   )
   > UpdateQrCodeVisibility()
   n3 = Event OnReceivedDoesIslandContainExperimentalDevices!() -> (OutputDelegate, then, bIslandContainsExperimentalDevices:bool) {is=title K2Node_Event}
   > SetVisibility(
    self=Get HBox_ExperimentalMessage() -> (HBox_ExperimentalMessage:Horizontal Box Object Reference)
    InVisibility=Select(Option 0:ESlateVisibility Enum=Collapsed, Option 1:ESlateVisibility Enum=SelfHitTestInvisible, Index=@n3.bIslandContainsExperimentalDevices) -> (ReturnValue:ESlateVisibility Enum)
   )
   Event OnQrCodeImageUpdated() {is=title K2Node_Event}
   > UpdateQrCodeVisibility()
   Event OnInitialized()
   > Set UIKitButton_PrimaryFunction(UIKitButton_PrimaryFunction:WBP UIKit Button CTA Base Object Reference=Cast To WBP_UIKit_ButtonCTA_Base!(Object=Get Button_PrimaryFunction()) -> (AsWBP UIKit Button CTA Base:WBP UIKit Button CTA Base Object Reference, bSuccess:bool=~) {is=title K2Node_DynamicCast}) -> (Output_Get:WBP UIKit Button CTA Base Object Reference)
   n4 = Event OnContentSwitcherActiveWidgetChanged!() -> (OutputDelegate, then, ActiveWidget:Widget Object Reference) {is=title K2Node_Event}
   n5 = Get Panel_CreateBuildInProgress() -> (Panel_CreateBuildInProgress:Panel Widget Object Reference)
   > SetVisibility(
    self=Get Button_PrimaryFunction()
    InVisibility=Select(Option 0=SelfHitTestInvisible, Option 1=Collapsed, Index="Equal (Object)"(A=@n5, B=@n4.ActiveWidget))
   )
   > UpdateButtonText!(execute, self, ActiveWidget:Widget Object Reference=@n4.ActiveWidget) -> (then)
   > SetClearPersistentDataInputEnabled!(
    execute
    self:Fort Creative Create Build Modal Object Reference
    bEnabled:bool="Equal (Object)"(A=Get Panel_CreateBuild() -> (Panel_CreateBuild:Panel Widget Object Reference), B=@n4.ActiveWidget)
   ) -> (then)
   > n6 = Get Intro Animation!(execute:Exec, ActiveWidget:Widget Object Reference=@n4.ActiveWidget) -> (IsValid:Exec, IsNotValid:Exec, Animation:Widget Animation Object Reference) {is=title K2Node_MacroInstance}
   > PlayAnimation(InAnimation=@n6)
   Event Reset() {is=title K2Node_Event}
   > StopAllAnimations(self:User Widget Object Reference)
   n7 = Get Switcher_Content() -> (Switcher_Content:Common Animated Switcher Object Reference)
   > SetDisableTransitionAnimation!(execute, self:Common Animated Switcher Object Reference=@n7, bDisableAnimation:bool=true) -> (then)
   > SetActiveWidgetIndex(self=@n7, Index=0)
   > SetDisableTransitionAnimation(self=@n7, bDisableAnimation=false)
   > SetRenderOpacity(self=Get Text_MnemonicCopied() -> (Text_MnemonicCopied:Common Text Object Reference), InOpacity=0)
   > SetVisibility(self=Get HBox_ClearPersistentDataWarning() -> (HBox_ClearPersistentDataWarning:Horizontal Box Object Reference), InVisibility=Hidden)
   > SetIsSelected!(execute, self:Common Button Base Object Reference=Get Button_ClearPersistentData(), InSelected:bool=false, bGiveClickFeedback:bool=false) -> (then)
   > SetVisibility(
    self=[Get HBox_ExperimentalMessage(), Get HBox_AudioMessage() -> (HBox_AudioMessage:Horizontal Box Object Reference)]
    InVisibility=Collapsed
   )
   Event OnMnemonicCopied() {is=title K2Node_Event}
   > PlayAnimation(InAnimation=Get MnemonicCopied() -> (MnemonicCopied:Widget Animation Object Reference))
   Event On Activated()
   > SetInitialContentPanel()
   Event OnClearPersistentDataSet() {is=title K2Node_Event}
   > SetVisibility(
    self=Get HBox_ClearPersistentDataWarning()
    InVisibility=Select(Option 0=Hidden, Option 1=SelfHitTestInvisible, Index=Get bClearPersistentData() -> (bClearPersistentData:bool))
   )
   n8 = Event OnReceivedDoesIslandContainFabricDevices!() -> (OutputDelegate, then, bIslandContainsFabricDevices:bool, bIslandContainsStreamingMusicFabricDevices:bool) {is=title K2Node_Event}
   > n9 = Sequence()
   n10 = NOT Boolean(A=@n8.bIslandContainsStreamingMusicFabricDevices)
   > SetVisibility(
    self=Get HBox_AudioMessage()
    InVisibility=Select(Option 0=Collapsed, Option 1=SelfHitTestInvisible, Index=AND Boolean(A=@n8.bIslandContainsFabricDevices, B=@n10))
   )
   SetVisibility(
    execute=@n9.then_1
    self=Get HBox_StreamingAudioMessage() -> (HBox_StreamingAudioMessage:Horizontal Box Object Reference)
    InVisibility=Select(Option 0=Collapsed, Option 1=SelfHitTestInvisible, Index=@n8.bIslandContainsStreamingMusicFabricDevices)
   )
  ]}
 GetIntroAnimation {nodes [
   n1 = Inputs!() -> (execute, ActiveWidget:Widget Object Reference)
   n2 = Select(
    Option 0:Widget Animation Object Reference
    Option 1:Widget Animation Object Reference=Get ShowUp() -> (ShowUp:Widget Animation Object Reference)
    Index="Equal (Object)"(A=@n1, B=Get Panel_CreateBuildInProgress())
   ) -> (ReturnValue:Widget Animation Object Reference)
   > n3 = Is Valid(InputObject=@n2)
   > Outputs!(IsValid:Exec, IsNotValid:Exec=@n3.Is Not Valid, Animation:Widget Animation Object Reference=@n2)
  ]}
 SetInitialContentPanel {nodes [
   SetInitialContentPanel() {is=title K2Node_FunctionEntry}
   > n1 = Cast To FortGameStateAthena!(execute, Object=GetGameState!() -> (ReturnValue:Game State Base Object Reference)) -> (then, CastFailed, AsFort Game State Athena:Fort Game State Athena Object Reference) {is=title K2Node_DynamicCast}
   n2 = GetOwningPlayerPawn!(self:User Widget Object Reference) -> (ReturnValue:Pawn Object Reference)
   > n3 = Is Valid(InputObject=@n2)
   > n4 = Branch(Condition=IsTherePersistenceDevices!(self:Fort Game State Athena Object Reference=@n1, SwitchingPawn:Pawn Object Reference=@n2) -> (ReturnValue:bool))
   > n5 = Set Local_ContentPanel(Local_ContentPanel:Panel Widget Object Reference=Get Panel_CreateBuild()) -> (Output_Get:Panel Widget Object Reference)
   n6 = Get Switcher_Content()
   SetDisableTransitionAnimation(execute=[@n7.then, @n5.then], self=@n6, bDisableAnimation=true)
   > SetActiveWidget(self:Widget Switcher Object Reference=@n6, Widget=Get Local_ContentPanel() -> (Local_ContentPanel:Panel Widget Object Reference))
   > SetDisableTransitionAnimation(self=@n6, bDisableAnimation=false)
   n7 = Set Local_ContentPanel(execute=[@n4.else, @n3.Is Not Valid, @n1.CastFailed], Local_ContentPanel=Get Panel_CreateBuildInProgress())
  ]}
 UpdateButtonText {nodes [
   UpdateButtonText!() -> (then, ActiveWidget:Widget Object Reference) {is=title K2Node_FunctionEntry}
   n1 = "Equal (Object)"(
    A=Get ActiveWidget() -> (ActiveWidget:Widget Object Reference)
    B=Get Panel_CreateBuildFailed() -> (Panel_CreateBuildFailed:Panel Widget Object Reference)
   )
   n2 = "Equal (Object)"(A=Get ActiveWidget(), B=Get Panel_CreateBuildSucceeded() -> (Panel_CreateBuildSucceeded:Panel Widget Object Reference))
   > SetText!(
    execute
    self:Fort CTAButton Object Reference=Get UIKitButton_PrimaryFunction() -> (UIKitButton_PrimaryFunction:WBP UIKit Button CTA Base Object Reference)
    ButtonText:"Text (by ref)"=Select(Option 0:Text="LOCTEXT(\"Create Private Code\")", Option 1:Text="LOCTEXT(\"Close\")", Index=OR Boolean(A=@n2, B=@n1)) -> (ReturnValue:Text)
   ) -> (then)
  ]}
 UpdateQrCodeVisibility {nodes [
   UpdateQrCodeVisibility() {is=title K2Node_FunctionEntry}
   > SetVisibility(
    self=Get Overlay_CreatorPortalQrCode() -> (Overlay_CreatorPortalQrCode:Overlay Object Reference)
    InVisibility=GetVisibility(self=Get LazyImage_CreatorPortalQrCode() -> (LazyImage_CreatorPortalQrCode:Fort Lazy Image Object Reference)) -> (ReturnValue=Visible)
   )
  ]}
]
variables [
 BacchusButton_MobileClose {type=CloseButton_C, default=null}
 UIKitButton_PrimaryFunction {type=WBP_UIKit_ButtonCTA_Base_C, default=null}
]
widgets {
 info {inheritedWidgetCount=13, parentClass=/Script/FortniteUI.FortCreativeCreateBuildModal, rootWidgetClass=&Overlay, widgetCount=72}
 widgets [
  Overlay_0 {widgetClassPath=&Overlay, widget=Overlay_0}
  CommonBorder_0 {
   widgetClassPath=&CommonBorder
   parent=Overlay_0
   widget=CommonBorder_0
   properties {
    verticalAlignment=VAlign_Center
    background {
     bIsDynamicallyLoaded=false
     drawAs=Image
     imageType=NoImage
     mirroring=NoMirror
     resourceName=None
     resourceObject=/Game/Athena/UI/Frontend/Art/M_UI_RadialGradientMaster_Dark
     tiling=NoTile
     imageSize=<32 32>
     margin=<ltrb 0 0 0 0>
     outlineSettings {
      bUseBrushTransparency=false
      roundingType=HalfHeightRadius
      width=0
      color {colorUseRule=UseColor_Specified, specifiedColor=<rgba 0 0 0 0>}
      cornerRadii=<0 0 0 0>
     }
     tintColor {colorUseRule=UseColor_Specified, specifiedColor=<rgba 1 1 1 1>}
     uVRegion {bIsValid=false, max=<0 0>, min=<0 0>}
    }
    padding=<ltrb 0 0 0 0>
   }
   slot {class=OverlaySlot, properties {horizontalAlignment=HAlign_Fill, verticalAlignment=VAlign_Fill}}
  }
  SizeBox_34 {
   widgetClassPath=&SizeBox
   parent=CommonBorder_0
   widget=SizeBox_34
   properties {bOverride_MaxDesiredHeight=true, bOverride_MinDesiredHeight=true, maxDesiredHeight=1000, minDesiredHeight=450}
   slot {class=BorderSlot, properties {verticalAlignment=VAlign_Center, padding=<ltrb 0 0 0 0>}}
  }
  Overlay_102 {widgetClassPath=&Overlay, parent=SizeBox_34, widget=Overlay_102, slot {class=SizeBoxSlot}}
  Containment {
   widgetClassPath=&Image
   parent=Overlay_102
   widget=Containment
   properties {
    visibility=SelfHitTestInvisible
    brush {
     bIsDynamicallyLoaded=false
     drawAs=Image
     imageType=NoImage
     mirroring=NoMirror
     resourceName=None
     resourceObject=/Game/Athena/UI/Subscriptions/Art/M_UI_Subs_Color_FIll_Modals
     tiling=NoTile
     imageSize=<32 32>
     margin=<ltrb 0 0 0 0>
     outlineSettings {
      bUseBrushTransparency=false
      roundingType=HalfHeightRadius
      width=0
      color {colorUseRule=UseColor_Specified, specifiedColor=<rgba 0 0 0 0>}
      cornerRadii=<0 0 0 0>
     }
     tintColor {colorUseRule=UseColor_Specified, specifiedColor=<rgba 1 1 1 1>}
     uVRegion {bIsValid=false, max=<0 0>, min=<0 0>}
    }
    colorAndOpacity=<rgba 1 1 1 0.5>
    renderTransformPivot=<0.5 1>
   }
   slot {class=OverlaySlot, properties {horizontalAlignment=HAlign_Fill, verticalAlignment=VAlign_Fill}}
  }
  Image_BottomLine {
   widgetClassPath=&Image
   parent=Overlay_102
   widget=Image_BottomLine
   properties {
    visibility=SelfHitTestInvisible
    brush {
     bIsDynamicallyLoaded=false
     drawAs=Image
     imageType=NoImage
     mirroring=NoMirror
     resourceName=None
     resourceObject=/FortUILibrary/Foundation/Art_ToolKit/AngledBoxes/M_UI_MoreInfo_BottomDividerLIne_Blue
     tiling=NoTile
     imageSize=<32 26>
     margin=<ltrb 0 0 0 0>
     outlineSettings {
      bUseBrushTransparency=false
      roundingType=HalfHeightRadius
      width=0
      color {colorUseRule=UseColor_Specified, specifiedColor=<rgba 0 0 0 0>}
      cornerRadii=<0 0 0 0>
     }
     tintColor {colorUseRule=UseColor_Specified, specifiedColor=<rgba 1 1 1 1>}
     uVRegion {bIsValid=false, max=<0 0>, min=<0 0>}
    }
   }
   slot {class=OverlaySlot, properties {horizontalAlignment=HAlign_Fill, verticalAlignment=VAlign_Bottom}}
  }
  Switcher_Content {
   widgetClassPath=/Script/CommonUI.CommonAnimatedSwitcher
   bInherited=true
   parent=Overlay_102
   widget=Switcher_Content
   slot {class=OverlaySlot, properties {horizontalAlignment=HAlign_Fill, verticalAlignment=VAlign_Fill, padding=<ltrb 128 0 128 0>}}
  }
  Panel_CreateBuild {widgetClassPath=&Overlay, bInherited=true, parent=Switcher_Content, widget=Panel_CreateBuild, slot {class=WidgetSwitcherSlot}}
  VerticalBox_88 {
   widgetClassPath=&VerticalBox
   parent=Panel_CreateBuild
   widget=VerticalBox_88
   slot {class=OverlaySlot, properties {horizontalAlignment=HAlign_Fill, verticalAlignment=VAlign_Fill, padding=<ltrb 0 88 0 0>}}
  }
  Title_1 {
   widgetClassPath=&CommonTextBlock
   parent=VerticalBox_88
   widget=Title_1
   properties {
    mobileFontSizeMultiplier=1.2
    text=Data Persistence Devices Found
    textTransformPolicy=ToUpper
    wrapTextAt=1152
    font {
     bForceMonospaced=false
     bMaterialIsStencil=false
     fontMaterial=null
     fontObject=&Burbank
     letterSpacing=0
     monospacedWidth=1
     size=25
     skewAmount=0
     typefaceFontName=BurbankRegular-Black
     outlineSettings {
      bApplyOutlineToDropShadows=false
      bMiteredCorners=false
      bSeparateFillAlpha=false
      outlineMaterial=null
      outlineSize=0
      outlineColor=<rgba 0 0 0 1>
     }
    }
    renderTransform {angle=0, scale=<1 1>, shear=<-12 0>, translation=<0 0>}
   }
   slot {class=VerticalBoxSlot, properties {padding=<ltrb 0 0 0 16>}}
  }
  CommonTextBlock_1 {
   widgetClassPath=&CommonTextBlock
   parent=VerticalBox_88
   widget=CommonTextBlock_1
   properties {
    autoWrapText=true
    lineHeightPercentage=1.29
    mobileFontSizeMultiplier=0.9
    style=/FortUILibrary/Styles/Typography/_TextStyle_Body1._TextStyle_Body1_C
    text=This island contains one or more persistence devices that may contain saved data from players.
    wrapTextAt=1100
    font {
     bForceMonospaced=false
     bMaterialIsStencil=false
     fontMaterial=null
     fontObject=&BurbankSmall
     letterSpacing=0
     monospacedWidth=1
     size=18
     skewAmount=0
     typefaceFontName=Bold
     outlineSettings {
      bApplyOutlineToDropShadows=false
      bMiteredCorners=false
      bSeparateFillAlpha=false
      outlineMaterial=null
      outlineSize=0
      outlineColor=<rgba 0 0 0 1>
     }
    }
    shadowOffset=<0 0>
    strikeBrush {
     bIsDynamicallyLoaded=false
     drawAs=NoDrawType
     imageType=NoImage
     mirroring=NoMirror
     resourceName=None
     resourceObject=null
     tiling=NoTile
     imageSize=<32 32>
     margin=<ltrb 0 0 0 0>
     outlineSettings {
      bUseBrushTransparency=false
      roundingType=HalfHeightRadius
      width=0
      color {colorUseRule=UseColor_Specified, specifiedColor=<rgba 0 0 0 0>}
      cornerRadii=<0 0 0 0>
     }
     tintColor {colorUseRule=UseColor_Specified, specifiedColor=<rgba 1 1 1 1>}
     uVRegion {bIsValid=false, max=<0 0>, min=<0 0>}
    }
   }
   slot {class=VerticalBoxSlot}
  }
  Spacer_86 {widgetClassPath=&Spacer, parent=VerticalBox_88, widget=Spacer_86, properties {size=<1 24>}, slot {class=VerticalBoxSlot}}
  HorizontalBox_0 {
   widgetClassPath=&HorizontalBox
   parent=VerticalBox_88
   widget=HorizontalBox_0
   slot {class=VerticalBoxSlot, properties {verticalAlignment=VAlign_Top, padding=<ltrb 0 0 0 20>, size {sizeRule=Fill, value=1}}}
  }
  VerticalBox_86 {
   widgetClassPath=&VerticalBox
   parent=HorizontalBox_0
   widget=VerticalBox_86
   slot {class=HorizontalBoxSlot, properties {size {sizeRule=Fill, value=1}}}
  }
  Text_ClearPersistentData {
   widgetClassPath=&CommonTextBlock
   parent=VerticalBox_86
   widget=Text_ClearPersistentData
   properties {
    lineHeightPercentage=1.29
    mobileFontSizeMultiplier=1.2
    style=&_TextStyle_Body1_C
    text=Clear All Player Save Data
    textTransformPolicy=ToUpper
    visibility=Collapsed
    font {
     bForceMonospaced=false
     bMaterialIsStencil=false
     fontMaterial=null
     fontObject=&BurbankSmall
     letterSpacing=0
     monospacedWidth=1
     size=18
     skewAmount=0
     typefaceFontName=Bold
     outlineSettings {
      bApplyOutlineToDropShadows=false
      bMiteredCorners=false
      bSeparateFillAlpha=false
      outlineMaterial=null
      outlineSize=0
      outlineColor=<rgba 0 0 0 1>
     }
    }
    shadowOffset=<0 0>
    strikeBrush {
     bIsDynamicallyLoaded=false
     drawAs=NoDrawType
     imageType=NoImage
     mirroring=NoMirror
     resourceName=None
     resourceObject=null
     tiling=NoTile
     imageSize=<32 32>
     margin=<ltrb 0 0 0 0>
     outlineSettings {
      bUseBrushTransparency=false
      roundingType=HalfHeightRadius
      width=0
      color {colorUseRule=UseColor_Specified, specifiedColor=<rgba 0 0 0 0>}
      cornerRadii=<0 0 0 0>
     }
     tintColor {colorUseRule=UseColor_Specified, specifiedColor=<rgba 1 1 1 1>}
     uVRegion {bIsValid=false, max=<0 0>, min=<0 0>}
    }
   }
   slot {class=VerticalBoxSlot}
  }
  Text_ClearPersistentData_1 {
   widgetClassPath=&CommonTextBlock
   parent=VerticalBox_86
   widget=Text_ClearPersistentData_1
   properties {
    lineHeightPercentage=1.29
    mobileFontSizeMultiplier=1.2
    style=&_TextStyle_Body1_C
    text=Check this box to reset player save data on your island.
    font {
     bForceMonospaced=false
     bMaterialIsStencil=false
     fontMaterial=null
     fontObject=&BurbankSmall
     letterSpacing=0
     monospacedWidth=1
     size=18
     skewAmount=0
     typefaceFontName=Bold
     outlineSettings {
      bApplyOutlineToDropShadows=false
      bMiteredCorners=false
      bSeparateFillAlpha=false
      outlineMaterial=null
      outlineSize=0
      outlineColor=<rgba 0 0 0 1>
     }
    }
    shadowOffset=<0 0>
    strikeBrush {
     bIsDynamicallyLoaded=false
     drawAs=NoDrawType
     imageType=NoImage
     mirroring=NoMirror
     resourceName=None
     resourceObject=null
     tiling=NoTile
     imageSize=<32 32>
     margin=<ltrb 0 0 0 0>
     outlineSettings {
      bUseBrushTransparency=false
      roundingType=HalfHeightRadius
      width=0
      color {colorUseRule=UseColor_Specified, specifiedColor=<rgba 0 0 0 0>}
      cornerRadii=<0 0 0 0>
     }
     tintColor {colorUseRule=UseColor_Specified, specifiedColor=<rgba 1 1 1 1>}
     uVRegion {bIsValid=false, max=<0 0>, min=<0 0>}
    }
   }
   slot {class=VerticalBoxSlot}
  }
  Button_ClearPersistentData {
   widgetClassPath=&WBP_UIKit_Button_Generic
   bIsVariable=true
   parent=VerticalBox_86
   widget=Button_ClearPersistentData
   properties {
    bInteractableWhenSelected=true
    bIsFocusable=false
    bSelectable=true
    bToggleable=true
    icon=&T-Icon-Confirmation_Empty-76
    text=Clear Player Saved Data
    triggeringInputAction {dataTable=&InputActionDataTable, rowName=GenericFaceButtonTop_Hold}
   }
   slot {class=VerticalBoxSlot, properties {horizontalAlignment=HAlign_Left, verticalAlignment=VAlign_Center, padding=<ltrb 0 8 0 12>}}
  }
  HBox_ClearPersistentDataWarning {
   widgetClassPath=&HorizontalBox
   bIsVariable=true
   parent=VerticalBox_86
   widget=HBox_ClearPersistentDataWarning
   slot {class=VerticalBoxSlot, properties {padding=<ltrb 28 0 0 0>}}
  }
  Image_96 {
   widgetClassPath=&Image
   parent=HBox_ClearPersistentDataWarning
   widget=Image_96
   properties {
    visibility=SelfHitTestInvisible
    brush {
     bIsDynamicallyLoaded=false
     drawAs=Image
     imageType=NoImage
     mirroring=NoMirror
     resourceName=None
     resourceObject=/Game/UI/Foundation/Shell/Textures/T-Icon-Warning-128
     tiling=NoTile
     imageSize=<32 32>
     margin=<ltrb 0 0 0 0>
     outlineSettings {
      bUseBrushTransparency=false
      roundingType=HalfHeightRadius
      width=0
      color {colorUseRule=UseColor_Specified, specifiedColor=<rgba 0 0 0 0>}
      cornerRadii=<0 0 0 0>
     }
     tintColor {colorUseRule=UseColor_Specified, specifiedColor=<rgba 1 1 1 1>}
     uVRegion {bIsValid=false, max=<0 0>, min=<0 0>}
    }
    colorAndOpacity=<rgba 0.9216 0.1812 0.2195 1>
   }
   slot {class=HorizontalBoxSlot, properties {horizontalAlignment=HAlign_Left, verticalAlignment=VAlign_Top, padding=<ltrb 0 6 6 0>}}
  }
  VerticalBox_0 {
   widgetClassPath=&VerticalBox
   parent=HBox_ClearPersistentDataWarning
   widget=VerticalBox_0
   slot {class=HorizontalBoxSlot, properties {verticalAlignment=VAlign_Center, padding=<ltrb 0 6 0 0>, size {sizeRule=Fill, value=1}}}
  }
  Text_ClearPersistentDataWarning {
   widgetClassPath=&CommonTextBlock
   parent=VerticalBox_0
   widget=Text_ClearPersistentDataWarning
   properties {
    mobileFontSizeMultiplier=1.2
    text=Warning: All Player data will be cleared
    textTransformPolicy=ToUpper
    colorAndOpacity {colorUseRule=UseColor_Specified, specifiedColor=<rgba 0.9216 0.1812 0.2195 1>}
    font {
     bForceMonospaced=false
     bMaterialIsStencil=false
     fontMaterial=null
     fontObject=&BurbankSmall
     letterSpacing=0
     monospacedWidth=1
     size=18
     skewAmount=0
     typefaceFontName=NotoSans-Regular
     outlineSettings {
      bApplyOutlineToDropShadows=false
      bMiteredCorners=false
      bSeparateFillAlpha=false
      outlineMaterial=null
      outlineSize=0
      outlineColor=<rgba 0 0 0 1>
     }
    }
   }
   slot {class=VerticalBoxSlot}
  }
  Text_ClearPersistentData_2 {
   widgetClassPath=&CommonTextBlock
   parent=VerticalBox_0
   widget=Text_ClearPersistentData_2
   properties {
    lineHeightPercentage=1.29
    mobileFontSizeMultiplier=1.2
    style=&_TextStyle_Body1_C
    text=Player data will not reset until the island is published.
    font {
     bForceMonospaced=false
     bMaterialIsStencil=false
     fontMaterial=null
     fontObject=&BurbankSmall
     letterSpacing=0
     monospacedWidth=1
     size=18
     skewAmount=0
     typefaceFontName=Bold
     outlineSettings {
      bApplyOutlineToDropShadows=false
      bMiteredCorners=false
      bSeparateFillAlpha=false
      outlineMaterial=null
      outlineSize=0
      outlineColor=<rgba 0 0 0 1>
     }
    }
    shadowOffset=<0 0>
    strikeBrush {
     bIsDynamicallyLoaded=false
     drawAs=NoDrawType
     imageType=NoImage
     mirroring=NoMirror
     resourceName=None
     resourceObject=null
     tiling=NoTile
     imageSize=<32 32>
     margin=<ltrb 0 0 0 0>
     outlineSettings {
      bUseBrushTransparency=false
      roundingType=HalfHeightRadius
      width=0
      color {colorUseRule=UseColor_Specified, specifiedColor=<rgba 0 0 0 0>}
      cornerRadii=<0 0 0 0>
     }
     tintColor {colorUseRule=UseColor_Specified, specifiedColor=<rgba 1 1 1 1>}
     uVRegion {bIsValid=false, max=<0 0>, min=<0 0>}
    }
   }
   slot {class=VerticalBoxSlot}
  }
  Spacer_1 {
   widgetClassPath=&Spacer
   parent=VerticalBox_88
   widget=Spacer_1
   slot {class=VerticalBoxSlot, properties {horizontalAlignment=HAlign_Left, verticalAlignment=VAlign_Top}}
  }
  Panel_CreateBuildInProgress {
   widgetClassPath=&Overlay
   bInherited=true
   parent=Switcher_Content
   widget=Panel_CreateBuildInProgress
   slot {class=WidgetSwitcherSlot}
  }
  Waiting_VBox {
   widgetClassPath=&VerticalBox
   parent=Panel_CreateBuildInProgress
   widget=Waiting_VBox
   slot {class=OverlaySlot, properties {horizontalAlignment=HAlign_Center, verticalAlignment=VAlign_Center}}
  }
  Title_3 {
   widgetClassPath=&CommonTextBlock
   parent=Waiting_VBox
   widget=Title_3
   properties {
    mobileFontSizeMultiplier=1.2
    text=Creating Private Code...
    textTransformPolicy=ToUpper
    wrapTextAt=1152
    font {
     bForceMonospaced=false
     bMaterialIsStencil=false
     fontMaterial=null
     fontObject=&Burbank
     letterSpacing=0
     monospacedWidth=1
     size=30
     skewAmount=0
     typefaceFontName=BurbankRegular-Black
     outlineSettings {
      bApplyOutlineToDropShadows=false
      bMiteredCorners=false
      bSeparateFillAlpha=false
      outlineMaterial=null
      outlineSize=0
      outlineColor=<rgba 1 1 1 0>
     }
    }
    renderTransform {angle=0, scale=<1 1>, shear=<-12 0>, translation=<0 0>}
    shadowOffset=<0 0>
   }
   slot {class=VerticalBoxSlot, properties {padding=<ltrb 0 0 0 16>}}
  }
  MatchmakingSpinner {
   widgetClassPath=&Image
   parent=Waiting_VBox
   widget=MatchmakingSpinner
   properties {
    visibility=SelfHitTestInvisible
    brush {
     bIsDynamicallyLoaded=false
     drawAs=Image
     imageType=NoImage
     mirroring=NoMirror
     resourceName=None
     resourceObject=/FortUILibrary/UIKit/Throbber/MI_UIKit_Monochrome_Throbber
     tiling=NoTile
     imageSize=<64 64>
     margin=<ltrb 0 0 0 0>
     outlineSettings {
      bUseBrushTransparency=false
      roundingType=HalfHeightRadius
      width=0
      color {colorUseRule=UseColor_Specified, specifiedColor=<rgba 0 0 0 0>}
      cornerRadii=<0 0 0 0>
     }
     tintColor {colorUseRule=UseColor_Specified, specifiedColor=<rgba 1 1 1 1>}
     uVRegion {bIsValid=false, max=<0 0>, min=<0 0>}
    }
   }
   slot {class=VerticalBoxSlot, properties {horizontalAlignment=HAlign_Center, verticalAlignment=VAlign_Center, padding=<ltrb 0 24 0 0>}}
  }
  Panel_CreateBuildSucceeded {
   widgetClassPath=&VerticalBox
   bInherited=true
   parent=Switcher_Content
   widget=Panel_CreateBuildSucceeded
   slot {class=WidgetSwitcherSlot}
  }
  Header_HBox {
   widgetClassPath=&HorizontalBox
   parent=Panel_CreateBuildSucceeded
   widget=Header_HBox
   slot {class=VerticalBoxSlot, properties {horizontalAlignment=HAlign_Left, padding=<ltrb 0 88 0 42>}}
  }
  Checkmark {
   widgetClassPath=&Image
   parent=Header_HBox
   widget=Checkmark
   properties {
    visibility=SelfHitTestInvisible
    brush {
     bIsDynamicallyLoaded=false
     drawAs=Image
     imageType=NoImage
     mirroring=NoMirror
     resourceName=None
     resourceObject=/Game/Athena/UI/InGame/Creative/PropertyEditor/Icon_Picker/Textures/128x128/T_UI_IconLibrary_checkmark_128
     tiling=NoTile
     imageSize=<100 100>
     margin=<ltrb 0 0 0 0>
     outlineSettings {
      bUseBrushTransparency=false
      roundingType=HalfHeightRadius
      width=0
      color {colorUseRule=UseColor_Specified, specifiedColor=<rgba 0 0 0 0>}
      cornerRadii=<0 0 0 0>
     }
     tintColor {colorUseRule=UseColor_Specified, specifiedColor=<rgba 0.0343 1 0.227 1>}
     uVRegion {bIsValid=false, max=<0 0>, min=<0 0>}
    }
   }
   slot {class=HorizontalBoxSlot, properties {horizontalAlignment=HAlign_Left, verticalAlignment=VAlign_Top, padding=<ltrb 0 0 24 0>}}
  }
  HeaderContent {widgetClassPath=&VerticalBox, parent=Header_HBox, widget=HeaderContent, slot {class=HorizontalBoxSlot}}
  Title {
   widgetClassPath=&CommonTextBlock
   parent=HeaderContent
   widget=Title
   properties {
    mobileFontSizeMultiplier=1.2
    text="Private Code Created!"
    textTransformPolicy=ToUpper
    wrapTextAt=1152
    font {
     bForceMonospaced=false
     bMaterialIsStencil=false
     fontMaterial=null
     fontObject=&Burbank
     letterSpacing=0
     monospacedWidth=1
     size=30
     skewAmount=0
     typefaceFontName=BurbankRegular-Black
     outlineSettings {
      bApplyOutlineToDropShadows=false
      bMiteredCorners=false
      bSeparateFillAlpha=false
      outlineMaterial=null
      outlineSize=0
      outlineColor=<rgba 1 1 1 0>
     }
    }
    renderTransform {angle=0, scale=<1 1>, shear=<-12 0>, translation=<0 0>}
    shadowOffset=<0 0>
   }
   slot {class=VerticalBoxSlot, properties {padding=<ltrb 0 0 0 16>}}
  }
  CommonTextBlock {
   widgetClassPath=&CommonTextBlock
   parent=HeaderContent
   widget=CommonTextBlock
   properties {
    autoWrapText=true
    lineHeightPercentage=1.29
    mobileFontSizeMultiplier=1.2
    style=&_TextStyle_Body1_C
    text=A new private code of your island has been created and is ready for publishing in the Creator Portal. From there you can add collaborators and submit to Discover.
    wrapTextAt=1100
    font {
     bForceMonospaced=false
     bMaterialIsStencil=false
     fontMaterial=null
     fontObject=&BurbankSmall
     letterSpacing=0
     monospacedWidth=1
     size=18
     skewAmount=0
     typefaceFontName=Bold
     outlineSettings {
      bApplyOutlineToDropShadows=false
      bMiteredCorners=false
      bSeparateFillAlpha=false
      outlineMaterial=null
      outlineSize=0
      outlineColor=<rgba 0 0 0 1>
     }
    }
    shadowOffset=<0 0>
    strikeBrush {
     bIsDynamicallyLoaded=false
     drawAs=NoDrawType
     imageType=NoImage
     mirroring=NoMirror
     resourceName=None
     resourceObject=null
     tiling=NoTile
     imageSize=<32 32>
     margin=<ltrb 0 0 0 0>
     outlineSettings {
      bUseBrushTransparency=false
      roundingType=HalfHeightRadius
      width=0
      color {colorUseRule=UseColor_Specified, specifiedColor=<rgba 0 0 0 0>}
      cornerRadii=<0 0 0 0>
     }
     tintColor {colorUseRule=UseColor_Specified, specifiedColor=<rgba 1 1 1 1>}
     uVRegion {bIsValid=false, max=<0 0>, min=<0 0>}
    }
   }
   slot {class=VerticalBoxSlot}
  }
  Content_HBox {
   widgetClassPath=&HorizontalBox
   parent=Panel_CreateBuildSucceeded
   widget=Content_HBox
   slot {class=VerticalBoxSlot, properties {horizontalAlignment=HAlign_Center}}
  }
  CreatorPortalInfo {
   widgetClassPath=&HorizontalBox
   parent=Content_HBox
   widget=CreatorPortalInfo
   slot {class=HorizontalBoxSlot, properties {horizontalAlignment=HAlign_Right, verticalAlignment=VAlign_Top, padding=<ltrb 0 0 60 0>}}
  }
  VBox_CreatorPortalUrl {
   widgetClassPath=&VerticalBox
   bIsVariable=true
   parent=CreatorPortalInfo
   widget=VBox_CreatorPortalUrl
   slot {class=HorizontalBoxSlot}
  }
  CommonTextBlock_287 {
   widgetClassPath=&CommonTextBlock
   parent=VBox_CreatorPortalUrl
   widget=CommonTextBlock_287
   properties {
    justification=Right
    lineHeightPercentage=1.29
    mobileFontSizeMultiplier=1.2
    style=&_TextStyle_Body1_C
    text=Go to the Creator Portal to publish.
    font {
     bForceMonospaced=false
     bMaterialIsStencil=false
     fontMaterial=null
     fontObject=&BurbankSmall
     letterSpacing=0
     monospacedWidth=1
     size=18
     skewAmount=0
     typefaceFontName=Bold
     outlineSettings {
      bApplyOutlineToDropShadows=false
      bMiteredCorners=false
      bSeparateFillAlpha=false
      outlineMaterial=null
      outlineSize=0
      outlineColor=<rgba 0 0 0 1>
     }
    }
    shadowOffset=<0 0>
    strikeBrush {
     bIsDynamicallyLoaded=false
     drawAs=NoDrawType
     imageType=NoImage
     mirroring=NoMirror
     resourceName=None
     resourceObject=null
     tiling=NoTile
     imageSize=<32 32>
     margin=<ltrb 0 0 0 0>
     outlineSettings {
      bUseBrushTransparency=false
      roundingType=HalfHeightRadius
      width=0
      color {colorUseRule=UseColor_Specified, specifiedColor=<rgba 0 0 0 0>}
      cornerRadii=<0 0 0 0>
     }
     tintColor {colorUseRule=UseColor_Specified, specifiedColor=<rgba 1 1 1 1>}
     uVRegion {bIsValid=false, max=<0 0>, min=<0 0>}
    }
   }
   slot {class=VerticalBoxSlot, properties {horizontalAlignment=HAlign_Right, verticalAlignment=VAlign_Top, padding=<ltrb 0 0 40 0>}}
  }
  Button_CreatorPortalUrl {
   widgetClassPath=&Button
   bInherited=true
   parent=VBox_CreatorPortalUrl
   widget=Button_CreatorPortalUrl
   properties {
    isFocusable=false
    backgroundColor=<rgba 1 1 1 0>
    widgetStyle {
     clickedSlateSound {resourceObject=null}
     disabled {
      bIsDynamicallyLoaded=false
      drawAs=NoDrawType
      imageType=NoImage
      mirroring=NoMirror
      resourceName=None
      resourceObject=null
      tiling=NoTile
      imageSize=<0 0>
      margin=<ltrb 0 0 0 0>
      outlineSettings {
       bUseBrushTransparency=false
       roundingType=HalfHeightRadius
       width=0
       color {colorUseRule=UseColor_Specified, specifiedColor=<rgba 0 0 0 0>}
       cornerRadii=<0 0 0 0>
      }
      tintColor {colorUseRule=UseColor_Specified, specifiedColor=<rgba 1 1 1 1>}
      uVRegion {bIsValid=false, max=<0 0>, min=<0 0>}
     }
     disabledForeground {colorUseRule=UseColor_Specified, specifiedColor=<rgba 0.5271 0.5271 0.5271 0>}
     hovered {
      bIsDynamicallyLoaded=false
      drawAs=NoDrawType
      imageType=NoImage
      mirroring=NoMirror
      resourceName=None
      resourceObject=null
      tiling=NoTile
      imageSize=<32 32>
      margin=<ltrb 0 0 0 0>
      outlineSettings {
       bUseBrushTransparency=true
       roundingType=FixedRadius
       width=1
       color {colorUseRule=UseColor_Specified, specifiedColor=<rgba 0.7243 0.7243 0.7243 1>}
       cornerRadii=<4 4 4 4>
      }
      tintColor {colorUseRule=UseColor_Specified, specifiedColor=<rgba 0.7243 0.7243 0.7243 1>}
      uVRegion {bIsValid=false, max=<0 0>, min=<0 0>}
     }
     hoveredForeground {colorUseRule=UseColor_Specified, specifiedColor=<rgba 0 1 1 0>}
     hoveredSlateSound {resourceObject=null}
     normal {
      bIsDynamicallyLoaded=false
      drawAs=NoDrawType
      imageType=NoImage
      mirroring=NoMirror
      resourceName=None
      resourceObject=null
      tiling=NoTile
      imageSize=<32 32>
      margin=<ltrb 0 0 0 0>
      outlineSettings {
       bUseBrushTransparency=true
       roundingType=FixedRadius
       width=1
       color {colorUseRule=UseColor_Specified, specifiedColor=<rgba 0.6951 0.6951 0.6951 1>}
       cornerRadii=<4 4 4 4>
      }
      tintColor {colorUseRule=UseColor_Specified, specifiedColor=<rgba 0.4955 0.4955 0.4955 1>}
      uVRegion {bIsValid=false, max=<0 0>, min=<0 0>}
     }
     normalForeground {colorUseRule=UseColor_Specified, specifiedColor=<rgba 0 1 1 0>}
     normalPadding=<ltrb 0 0 0 0>
     pressed {
      bIsDynamicallyLoaded=false
      drawAs=NoDrawType
      imageType=NoImage
      mirroring=NoMirror
      resourceName=None
      resourceObject=null
      tiling=NoTile
      imageSize=<32 32>
      margin=<ltrb 0 0 0 0>
      outlineSettings {
       bUseBrushTransparency=true
       roundingType=FixedRadius
       width=1
       color {colorUseRule=UseColor_Specified, specifiedColor=<rgba 0.7243 0.7243 0.7243 1>}
       cornerRadii=<4 4 4 4>
      }
      tintColor {colorUseRule=UseColor_Specified, specifiedColor=<rgba 0.3843 0.3843 0.3843 1>}
      uVRegion {bIsValid=false, max=<0 0>, min=<0 0>}
     }
     pressedForeground {colorUseRule=UseColor_Specified, specifiedColor=<rgba 0 1 1 0>}
     pressedPadding=<ltrb 0 0 0 0>
     pressedSlateSound {resourceObject=null}
    }
   }
   slot {class=VerticalBoxSlot, properties {horizontalAlignment=HAlign_Right, verticalAlignment=VAlign_Top, padding=<ltrb 0 0 40 0>}}
   bIsVariable=~
  }
  Text_CreatorPortalUrl {
   widgetClassPath=&CommonTextBlock
   bInherited=true
   parent=Button_CreatorPortalUrl
   widget=Text_CreatorPortalUrl
   properties {
    bAutoCollapseWithEmptyText=true
    justification=Right
    mobileFontSizeMultiplier=1.2
    text=https://create.fortnite.com
    colorAndOpacity {colorUseRule=UseColor_Specified, specifiedColor=<rgba 0.0549 0.6431 0.9216 1>}
    font {
     bForceMonospaced=false
     bMaterialIsStencil=false
     fontMaterial=null
     fontObject=&BurbankSmall
     letterSpacing=10
     monospacedWidth=1
     size=18
     skewAmount=0
     typefaceFontName=Bold
     outlineSettings {
      bApplyOutlineToDropShadows=false
      bMiteredCorners=false
      bSeparateFillAlpha=false
      outlineMaterial=null
      outlineSize=0
      outlineColor=<rgba 0 0 0 1>
     }
    }
    shadowOffset=<0 0>
   }
   slot {class=ButtonSlot, properties {padding=<ltrb 0 0 0 0>}}
  }
  Overlay_CreatorPortalQrCode {
   widgetClassPath=&Overlay
   bIsVariable=true
   parent=CreatorPortalInfo
   widget=Overlay_CreatorPortalQrCode
   slot {class=HorizontalBoxSlot}
  }
  QRBacking {
   widgetClassPath=&Image
   parent=Overlay_CreatorPortalQrCode
   widget=QRBacking
   properties {
    visibility=SelfHitTestInvisible
    brush {
     bIsDynamicallyLoaded=false
     drawAs=Image
     imageType=NoImage
     mirroring=NoMirror
     resourceName=None
     resourceObject=/FortniteCoreUI/Athena/UI/Frontend/Spatial/Art/M_Ui_Skewed_QRBorder
     tiling=NoTile
     imageSize=<32 32>
     margin=<ltrb 0 0 0 0>
     outlineSettings {
      bUseBrushTransparency=false
      roundingType=HalfHeightRadius
      width=0
      color {colorUseRule=UseColor_Specified, specifiedColor=<rgba 0 0 0 0>}
      cornerRadii=<0 0 0 0>
     }
     tintColor {colorUseRule=UseColor_Specified, specifiedColor=<rgba 1 1 1 1>}
     uVRegion {bIsValid=false, max=<0 0>, min=<0 0>}
    }
   }
   slot {class=OverlaySlot, properties {horizontalAlignment=HAlign_Fill, verticalAlignment=VAlign_Fill, padding=<ltrb -16 -16 -16 -16>}}
  }
  LazyImage_CreatorPortalQrCode {
   widgetClassPath=&FortLazyImage
   bInherited=true
   parent=Overlay_CreatorPortalQrCode
   widget=LazyImage_CreatorPortalQrCode
   properties {brush {
     bIsDynamicallyLoaded=false
     drawAs=Image
     imageType=NoImage
     mirroring=NoMirror
     resourceName=None
     resourceObject=null
     tiling=NoTile
     imageSize=<128 128>
     margin=<ltrb 0 0 0 0>
     outlineSettings {
      bUseBrushTransparency=false
      roundingType=HalfHeightRadius
      width=0
      color {colorUseRule=UseColor_Specified, specifiedColor=<rgba 0 0 0 0>}
      cornerRadii=<0 0 0 0>
     }
     tintColor {colorUseRule=UseColor_Specified, specifiedColor=<rgba 1 1 1 1>}
     uVRegion {bIsValid=false, max=<0 0>, min=<0 0>}
    }}
   slot {class=OverlaySlot, properties {horizontalAlignment=HAlign_Center, verticalAlignment=VAlign_Center}}
  }
  Divider {
   widgetClassPath=&Image
   parent=Content_HBox
   widget=Divider
   properties {
    visibility=SelfHitTestInvisible
    brush {
     bIsDynamicallyLoaded=false
     drawAs=Image
     imageType=NoImage
     mirroring=NoMirror
     resourceName=None
     resourceObject=null
     tiling=NoTile
     imageSize=<6 200>
     margin=<ltrb 0 0 0 0>
     outlineSettings {
      bUseBrushTransparency=false
      roundingType=HalfHeightRadius
      width=0
      color {colorUseRule=UseColor_Specified, specifiedColor=<rgba 0 0 0 0>}
      cornerRadii=<0 0 0 0>
     }
     tintColor {colorUseRule=UseColor_Specified, specifiedColor=<rgba 1 1 1 1>}
     uVRegion {bIsValid=false, max=<0 0>, min=<0 0>}
    }
    colorAndOpacity=<rgba 1 1 1 0.2>
   }
   slot {class=HorizontalBoxSlot}
  }
  BuildCode {
   widgetClassPath=&VerticalBox
   parent=Content_HBox
   widget=BuildCode
   slot {class=HorizontalBoxSlot, properties {padding=<ltrb 60 0 0 0>}}
  }
  CommonTextBlock_151 {
   widgetClassPath=&CommonTextBlock
   parent=BuildCode
   widget=CommonTextBlock_151
   properties {
    lineHeightPercentage=1.29
    mobileFontSizeMultiplier=1.2
    style=&_TextStyle_Body1_C
    text=You can test your island using this private code:
    font {
     bForceMonospaced=false
     bMaterialIsStencil=false
     fontMaterial=null
     fontObject=&BurbankSmall
     letterSpacing=0
     monospacedWidth=1
     size=18
     skewAmount=0
     typefaceFontName=Bold
     outlineSettings {
      bApplyOutlineToDropShadows=false
      bMiteredCorners=false
      bSeparateFillAlpha=false
      outlineMaterial=null
      outlineSize=0
      outlineColor=<rgba 0 0 0 1>
     }
    }
    shadowOffset=<0 0>
    strikeBrush {
     bIsDynamicallyLoaded=false
     drawAs=NoDrawType
     imageType=NoImage
     mirroring=NoMirror
     resourceName=None
     resourceObject=null
     tiling=NoTile
     imageSize=<32 32>
     margin=<ltrb 0 0 0 0>
     outlineSettings {
      bUseBrushTransparency=false
      roundingType=HalfHeightRadius
      width=0
      color {colorUseRule=UseColor_Specified, specifiedColor=<rgba 0 0 0 0>}
      cornerRadii=<0 0 0 0>
     }
     tintColor {colorUseRule=UseColor_Specified, specifiedColor=<rgba 1 1 1 1>}
     uVRegion {bIsValid=false, max=<0 0>, min=<0 0>}
    }
   }
   slot {class=VerticalBoxSlot, properties {padding=<ltrb 0 0 0 10>}}
  }
  Text_Mnemonic {
   widgetClassPath=&CommonTextBlock
   bInherited=true
   parent=BuildCode
   widget=Text_Mnemonic
   properties {
    mobileFontSizeMultiplier=1.2
    style=/FortUILibrary/Styles/Typography/_TextStyle_SubH2._TextStyle_SubH2_C
    text="{Private Code}"
    font {
     bForceMonospaced=false
     bMaterialIsStencil=false
     fontMaterial=null
     fontObject=&Burbank
     letterSpacing=0
     monospacedWidth=1
     size=33
     skewAmount=0
     typefaceFontName=BurbankRegular-Black
     outlineSettings {
      bApplyOutlineToDropShadows=false
      bMiteredCorners=false
      bSeparateFillAlpha=false
      outlineMaterial=null
      outlineSize=0
      outlineColor=<rgba 0 0 0 1>
     }
    }
    renderTransform {angle=0, scale=<1 1>, shear=<-12 0>, translation=<0 0>}
    shadowOffset=<0 0>
    strikeBrush {
     bIsDynamicallyLoaded=false
     drawAs=NoDrawType
     imageType=NoImage
     mirroring=NoMirror
     resourceName=None
     resourceObject=null
     tiling=NoTile
     imageSize=<32 32>
     margin=<ltrb 0 0 0 0>
     outlineSettings {
      bUseBrushTransparency=false
      roundingType=HalfHeightRadius
      width=0
      color {colorUseRule=UseColor_Specified, specifiedColor=<rgba 0 0 0 0>}
      cornerRadii=<0 0 0 0>
     }
     tintColor {colorUseRule=UseColor_Specified, specifiedColor=<rgba 1 1 1 1>}
     uVRegion {bIsValid=false, max=<0 0>, min=<0 0>}
    }
   }
   slot {class=VerticalBoxSlot, properties {horizontalAlignment=HAlign_Left, verticalAlignment=VAlign_Top, padding=<ltrb 0 0 0 10>}}
  }
  Button_CopyMnemonic {
   widgetClassPath=/Game/Valkyrie/UMG/UEFN_Button_Quiet
   bInherited=true
   parent=BuildCode
   widget=Button_CopyMnemonic
   properties {
    text=Copy Private Code
    hoveredSlateSoundOverride {resourceObject=/Game/Sounds/Fort_UI_Sounds/FrontEnd/New_FrontEnd/Fortnite_UI_MissionObj_Hover_01_Cue}
    pressedSlateSoundOverride {resourceObject=/Game/Sounds/Fort_UI_Sounds/FrontEnd/New_FrontEnd/Fortnite_UI_MainMenu_Back_01_Cue}
    triggeringInputAction {dataTable=&InputActionDataTable, rowName=GenericAccept}
   }
   slot {class=VerticalBoxSlot, properties {horizontalAlignment=HAlign_Left, verticalAlignment=VAlign_Bottom, padding=<ltrb 0 10 0 10>}}
  }
  Text_MnemonicCopied {
   widgetClassPath=&CommonTextBlock
   bIsVariable=true
   parent=BuildCode
   widget=Text_MnemonicCopied
   properties {
    mobileFontSizeMultiplier=1.2
    renderOpacity=0
    text="Private code copied to clipboard!"
    textTransformPolicy=ToUpper
    colorAndOpacity {colorUseRule=UseColor_Specified, specifiedColor=<rgba 0.9301 1 0.0097 1>}
    font {
     bForceMonospaced=false
     bMaterialIsStencil=false
     fontMaterial=null
     fontObject=&BurbankSmall
     letterSpacing=0
     monospacedWidth=1
     size=18
     skewAmount=0
     typefaceFontName=Black
     outlineSettings {
      bApplyOutlineToDropShadows=false
      bMiteredCorners=false
      bSeparateFillAlpha=false
      outlineMaterial=null
      outlineSize=0
      outlineColor=<rgba 0 0 0 1>
     }
    }
   }
   slot {class=VerticalBoxSlot, properties {padding=<ltrb 0 12 0 12>}}
  }
  VerticalBox_3 {
   widgetClassPath=&VerticalBox
   parent=Panel_CreateBuildSucceeded
   widget=VerticalBox_3
   slot {class=VerticalBoxSlot, properties {padding=<ltrb 0 12 0 12>}}
  }
  HBox_AudioMessage {
   widgetClassPath=&HorizontalBox
   bIsVariable=true
   parent=VerticalBox_3
   widget=HBox_AudioMessage
   properties {visibility=Collapsed}
   slot {class=VerticalBoxSlot}
  }
  Image {
   widgetClassPath=&Image
   parent=HBox_AudioMessage
   widget=Image
   properties {
    visibility=SelfHitTestInvisible
    brush {
     bIsDynamicallyLoaded=false
     drawAs=Image
     imageType=NoImage
     mirroring=NoMirror
     resourceName=None
     resourceObject=/FortUILibrary/Textures/MaterialInstances/MI_UIKit_Icon_Information
     tiling=NoTile
     imageSize=<32 32>
     margin=<ltrb 0 0 0 0>
     outlineSettings {
      bUseBrushTransparency=false
      roundingType=HalfHeightRadius
      width=0
      color {colorUseRule=UseColor_Specified, specifiedColor=<rgba 0 0 0 0>}
      cornerRadii=<0 0 0 0>
     }
     tintColor {colorUseRule=UseColor_Specified, specifiedColor=<rgba 1 1 1 0.8>}
     uVRegion {bIsValid=false, max=<0 0>, min=<0 0>}
    }
   }
   slot {class=HorizontalBoxSlot, properties {horizontalAlignment=HAlign_Left, verticalAlignment=VAlign_Top, padding=<ltrb 0 12 12 0>}}
  }
  VerticalBox {
   widgetClassPath=&VerticalBox
   parent=HBox_AudioMessage
   widget=VerticalBox
   slot {class=HorizontalBoxSlot, properties {verticalAlignment=VAlign_Center, padding=<ltrb 0 6 0 0>, size {sizeRule=Fill, value=1}}}
  }
  Text_ClearPersistentDataWarning_1 {
   widgetClassPath=&CommonTextBlock
   parent=VerticalBox
   widget=Text_ClearPersistentDataWarning_1
   properties {
    autoWrapText=true
    mobileFontSizeMultiplier=1.2
    text="Please be reminded that use of a patchwork audio device is limited to the creation of new and original musical works. Use of an audio device to re-create copyrighted musical compositions owned by third parties is a violation of Epic’s terms and may result in your island being unpublished."
    colorAndOpacity {colorUseRule=UseColor_Specified, specifiedColor=<rgba 1 1 1 0.8>}
    font {
     bForceMonospaced=false
     bMaterialIsStencil=false
     fontMaterial=null
     fontObject=&BurbankSmall
     letterSpacing=0
     monospacedWidth=1
     size=18
     skewAmount=0
     typefaceFontName=NotoSans-Regular
     outlineSettings {
      bApplyOutlineToDropShadows=false
      bMiteredCorners=false
      bSeparateFillAlpha=false
      outlineMaterial=null
      outlineSize=0
      outlineColor=<rgba 0 0 0 1>
     }
    }
   }
   slot {class=VerticalBoxSlot}
  }
  HBox_StreamingAudioMessage {
   widgetClassPath=&HorizontalBox
   bIsVariable=true
   parent=VerticalBox_3
   widget=HBox_StreamingAudioMessage
   properties {visibility=Collapsed}
   slot {class=VerticalBoxSlot}
  }
  Image_2 {
   widgetClassPath=&Image
   parent=HBox_StreamingAudioMessage
   widget=Image_2
   properties {
    visibility=SelfHitTestInvisible
    brush {
     bIsDynamicallyLoaded=false
     drawAs=Image
     imageType=NoImage
     mirroring=NoMirror
     resourceName=None
     resourceObject=&MI_UIKit_Icon_Information
     tiling=NoTile
     imageSize=<32 32>
     margin=<ltrb 0 0 0 0>
     outlineSettings {
      bUseBrushTransparency=false
      roundingType=HalfHeightRadius
      width=0
      color {colorUseRule=UseColor_Specified, specifiedColor=<rgba 0 0 0 0>}
      cornerRadii=<0 0 0 0>
     }
     tintColor {colorUseRule=UseColor_Specified, specifiedColor=<rgba 1 1 1 0.8>}
     uVRegion {bIsValid=false, max=<0 0>, min=<0 0>}
    }
   }
   slot {class=HorizontalBoxSlot, properties {horizontalAlignment=HAlign_Left, verticalAlignment=VAlign_Top, padding=<ltrb 0 12 12 0>}}
  }
  VerticalBox_2 {
   widgetClassPath=&VerticalBox
   parent=HBox_StreamingAudioMessage
   widget=VerticalBox_2
   slot {class=HorizontalBoxSlot, properties {verticalAlignment=VAlign_Center, padding=<ltrb 0 6 0 0>, size {sizeRule=Fill, value=1}}}
  }
  Text_ClearPersistentDataWarning_3 {
   widgetClassPath=&CommonTextBlock
   parent=VerticalBox_2
   widget=Text_ClearPersistentDataWarning_3
   properties {
    autoWrapText=true
    mobileFontSizeMultiplier=1.2
    text="Please be reminded that use of a patchwork streaming audio devices should only be used to create musical gameplay experiences per the policies outlined in the Fortnite Island Creator Rules. Violating Epic’s terms may result in your island being unpublished."
    colorAndOpacity {colorUseRule=UseColor_Specified, specifiedColor=<rgba 1 1 1 0.8>}
    font {
     bForceMonospaced=false
     bMaterialIsStencil=false
     fontMaterial=null
     fontObject=&BurbankSmall
     letterSpacing=0
     monospacedWidth=1
     size=18
     skewAmount=0
     typefaceFontName=NotoSans-Regular
     outlineSettings {
      bApplyOutlineToDropShadows=false
      bMiteredCorners=false
      bSeparateFillAlpha=false
      outlineMaterial=null
      outlineSize=0
      outlineColor=<rgba 0 0 0 1>
     }
    }
   }
   slot {class=VerticalBoxSlot}
  }
  HBox_ExperimentalMessage {
   widgetClassPath=&HorizontalBox
   bIsVariable=true
   parent=VerticalBox_3
   widget=HBox_ExperimentalMessage
   properties {visibility=Collapsed}
   slot {class=VerticalBoxSlot}
  }
  Image_1 {
   widgetClassPath=&Image
   parent=HBox_ExperimentalMessage
   widget=Image_1
   properties {
    visibility=SelfHitTestInvisible
    brush {
     bIsDynamicallyLoaded=false
     drawAs=Image
     imageType=NoImage
     mirroring=NoMirror
     resourceName=None
     resourceObject=&MI_UIKit_Icon_Information
     tiling=NoTile
     imageSize=<32 32>
     margin=<ltrb 0 0 0 0>
     outlineSettings {
      bUseBrushTransparency=false
      roundingType=HalfHeightRadius
      width=0
      color {colorUseRule=UseColor_Specified, specifiedColor=<rgba 0 0 0 0>}
      cornerRadii=<0 0 0 0>
     }
     tintColor {colorUseRule=UseColor_Specified, specifiedColor=<rgba 1 1 1 0.8>}
     uVRegion {bIsValid=false, max=<0 0>, min=<0 0>}
    }
   }
   slot {class=HorizontalBoxSlot, properties {horizontalAlignment=HAlign_Left, verticalAlignment=VAlign_Top, padding=<ltrb 0 12 12 0>}}
  }
  VerticalBox_1 {
   widgetClassPath=&VerticalBox
   parent=HBox_ExperimentalMessage
   widget=VerticalBox_1
   slot {class=HorizontalBoxSlot, properties {verticalAlignment=VAlign_Center, padding=<ltrb 0 6 0 0>, size {sizeRule=Fill, value=1}}}
  }
  Text_ClearPersistentDataWarning_2 {
   widgetClassPath=&CommonTextBlock
   parent=VerticalBox_1
   widget=Text_ClearPersistentDataWarning_2
   properties {
    autoWrapText=true
    mobileFontSizeMultiplier=1.2
    text="Islands with experimental assets are testable with private codes, but will not be publishable."
    colorAndOpacity {colorUseRule=UseColor_Specified, specifiedColor=<rgba 1 1 1 0.8>}
    font {
     bForceMonospaced=false
     bMaterialIsStencil=false
     fontMaterial=null
     fontObject=&BurbankSmall
     letterSpacing=0
     monospacedWidth=1
     size=18
     skewAmount=0
     typefaceFontName=NotoSans-Regular
     outlineSettings {
      bApplyOutlineToDropShadows=false
      bMiteredCorners=false
      bSeparateFillAlpha=false
      outlineMaterial=null
      outlineSize=0
      outlineColor=<rgba 0 0 0 1>
     }
    }
   }
   slot {class=VerticalBoxSlot}
  }
  Spacer {widgetClassPath=&Spacer, parent=Panel_CreateBuildSucceeded, widget=Spacer, properties {size=<0 64>}, slot {class=VerticalBoxSlot}}
  Panel_CreateBuildFailed {
   widgetClassPath=&Overlay
   bInherited=true
   parent=Switcher_Content
   widget=Panel_CreateBuildFailed
   slot {class=WidgetSwitcherSlot}
  }
  HorizontalBox_238 {
   widgetClassPath=&HorizontalBox
   parent=Panel_CreateBuildFailed
   widget=HorizontalBox_238
   slot {class=OverlaySlot, properties {horizontalAlignment=HAlign_Fill, verticalAlignment=VAlign_Fill}}
  }
  Header_1 {
   widgetClassPath=&VerticalBox
   parent=HorizontalBox_238
   widget=Header_1
   slot {class=HorizontalBoxSlot, properties {horizontalAlignment=HAlign_Left, verticalAlignment=VAlign_Center}}
  }
  Title_2 {
   widgetClassPath=&CommonTextBlock
   parent=Header_1
   widget=Title_2
   properties {
    mobileFontSizeMultiplier=1.2
    text=Private Code Failed
    textTransformPolicy=ToUpper
    wrapTextAt=1152
    font {
     bForceMonospaced=false
     bMaterialIsStencil=false
     fontMaterial=null
     fontObject=&Burbank
     letterSpacing=0
     monospacedWidth=1
     size=30
     skewAmount=0
     typefaceFontName=BurbankRegular-Black
     outlineSettings {
      bApplyOutlineToDropShadows=false
      bMiteredCorners=false
      bSeparateFillAlpha=false
      outlineMaterial=null
      outlineSize=0
      outlineColor=<rgba 1 1 1 0>
     }
    }
    renderTransform {angle=0, scale=<1 1>, shear=<-12 0>, translation=<0 0>}
    shadowOffset=<0 0>
   }
   slot {class=VerticalBoxSlot, properties {padding=<ltrb 0 0 0 16>}}
  }
  CommonRichTextBlock_1 {
   widgetClassPath=&CommonRichTextBlock
   parent=Header_1
   widget=CommonRichTextBlock_1
   properties {
    autoWrapText=true
    bIsScrollingEnabled=false
    bOverrideDefaultStyle=true
    defaultTextStyleOverrideClass=&_TextStyle_Body1_C
    mobileTextBlockScale=1.2
    text=There was an error while creating your private code. Fix the error and try again.
    wrapTextAt=1050
    defaultTextStyleOverride {
     overflowPolicy=Clip
     transformPolicy=None
     colorAndOpacity {colorUseRule=UseColor_Specified, specifiedColor=<rgba 1 1 1 1>}
     font {
      bForceMonospaced=false
      bMaterialIsStencil=false
      fontMaterial=null
      fontObject=&NotoSans
      letterSpacing=0
      monospacedWidth=1
      size=24
      skewAmount=0
      typefaceFontName=NotoSans-Regular
      outlineSettings {
       bApplyOutlineToDropShadows=false
       bMiteredCorners=false
       bSeparateFillAlpha=false
       outlineMaterial=null
       outlineSize=0
       outlineColor=<rgba 0 0 0 1>
      }
     }
     highlightColor {colorUseRule=UseColor_Specified, specifiedColor=<rgba 0 0 0 1>}
     highlightShape {
      bIsDynamicallyLoaded=false
      drawAs=Image
      imageType=NoImage
      mirroring=NoMirror
      resourceName=None
      resourceObject=null
      tiling=NoTile
      imageSize=<32 32>
      margin=<ltrb 0 0 0 0>
      outlineSettings {
       bUseBrushTransparency=false
       roundingType=HalfHeightRadius
       width=0
       color {colorUseRule=UseColor_Specified, specifiedColor=<rgba 0 0 0 0>}
       cornerRadii=<0 0 0 0>
      }
      tintColor {colorUseRule=UseColor_Specified, specifiedColor=<rgba 1 1 1 1>}
      uVRegion {bIsValid=false, max=<0 0>, min=<0 0>}
     }
     selectedBackgroundColor {colorUseRule=UseColor_Foreground, specifiedColor=<rgba 1 0 1 1>}
     shadowColorAndOpacity=<rgba 0 0 0 1>
     shadowOffset=<0 0>
     strikeBrush {
      bIsDynamicallyLoaded=false
      drawAs=Image
      imageType=NoImage
      mirroring=NoMirror
      resourceName=None
      resourceObject=null
      tiling=NoTile
      imageSize=<32 32>
      margin=<ltrb 0 0 0 0>
      outlineSettings {
       bUseBrushTransparency=false
       roundingType=HalfHeightRadius
       width=0
       color {colorUseRule=UseColor_Specified, specifiedColor=<rgba 0 0 0 0>}
       cornerRadii=<0 0 0 0>
      }
      tintColor {colorUseRule=UseColor_Specified, specifiedColor=<rgba 1 1 1 1>}
      uVRegion {bIsValid=false, max=<0 0>, min=<0 0>}
     }
     underlineBrush {
      bIsDynamicallyLoaded=false
      drawAs=Image
      imageType=NoImage
      mirroring=NoMirror
      resourceName=None
      resourceObject=null
      tiling=NoTile
      imageSize=<32 32>
      margin=<ltrb 0 0 0 0>
      outlineSettings {
       bUseBrushTransparency=false
       roundingType=HalfHeightRadius
       width=0
       color {colorUseRule=UseColor_Specified, specifiedColor=<rgba 0 0 0 0>}
       cornerRadii=<0 0 0 0>
      }
      tintColor {colorUseRule=UseColor_Specified, specifiedColor=<rgba 1 1 1 1>}
      uVRegion {bIsValid=false, max=<0 0>, min=<0 0>}
     }
    }
   }
   slot {class=VerticalBoxSlot}
  }
  Text_CreateBuildError {
   widgetClassPath=&CommonTextBlock
   bInherited=true
   parent=Header_1
   widget=Text_CreateBuildError
   properties {
    mobileFontSizeMultiplier=1.2
    text="{Error Text}"
    colorAndOpacity {colorUseRule=UseColor_Specified, specifiedColor=<rgba 0.9301 0.0742 0.107 1>}
    font {
     bForceMonospaced=false
     bMaterialIsStencil=false
     fontMaterial=null
     fontObject=&BurbankSmall
     letterSpacing=0
     monospacedWidth=1
     size=18
     skewAmount=0
     typefaceFontName=Black
     outlineSettings {
      bApplyOutlineToDropShadows=false
      bMiteredCorners=false
      bSeparateFillAlpha=false
      outlineMaterial=null
      outlineSize=0
      outlineColor=<rgba 0 0 0 1>
     }
    }
    shadowOffset=<0 0>
   }
   slot {class=VerticalBoxSlot, properties {padding=<ltrb 0 48 0 0>}}
  }
  Icon_Warning {
   widgetClassPath=&Image
   parent=HorizontalBox_238
   widget=Icon_Warning
   properties {
    visibility=SelfHitTestInvisible
    brush {
     bIsDynamicallyLoaded=false
     drawAs=Image
     imageType=NoImage
     mirroring=NoMirror
     resourceName=None
     resourceObject=&T-Icon-Warning-128
     tiling=NoTile
     imageSize=<312 312>
     margin=<ltrb 0 0 0 0>
     outlineSettings {
      bUseBrushTransparency=false
      roundingType=HalfHeightRadius
      width=0
      color {colorUseRule=UseColor_Specified, specifiedColor=<rgba 0 0 0 0>}
      cornerRadii=<0 0 0 0>
     }
     tintColor {colorUseRule=UseColor_Specified, specifiedColor=<rgba 1 1 1 0.4>}
     uVRegion {bIsValid=false, max=<0 0>, min=<0 0>}
    }
    colorAndOpacity=<rgba 0.2 0.6627 1 0.6>
   }
   slot {class=HorizontalBoxSlot, properties {horizontalAlignment=HAlign_Right, verticalAlignment=VAlign_Center, padding=<ltrb 200 0 0 0>}}
  }
  Button_PrimaryFunction {
   widgetClassPath=&WBP_UIKit_Button_Regular
   bInherited=true
   parent=Overlay_102
   widget=Button_PrimaryFunction
   properties {
    bIsFocusable=false
    minHeight=54
    text=Close
    hoveredSlateSoundOverride {resourceObject=&Fortnite_UI_MissionObj_Hover_01_Cue}
    pressedSlateSoundOverride {resourceObject=/Game/Sounds/Fort_UI_Sounds/FrontEnd/New_FrontEnd/Fortnite_UI_MissionObj_Close_01_Cue}
    triggeringInputAction {dataTable=&InputActionDataTable, rowName=GenericForward}
   }
   slot {class=OverlaySlot, properties {horizontalAlignment=HAlign_Center, verticalAlignment=VAlign_Bottom, padding=<ltrb 0 0 0 -14>}}
  }
  VisibilityBorder_Touch {
   widgetClassPath=&CommonHardwareVisibilityBorder
   parent=Overlay_102
   widget=VisibilityBorder_Touch
   properties {visibilityQuery {
     autoDescription=" ANY( Input.Touch )"
     tokenStreamVersion=0
     userDescription=""
     queryTokenStream [0,1,1,1,0]
     tagDictionary [{tagName=Input.Touch}]
    }}
   slot {class=OverlaySlot, properties {horizontalAlignment=HAlign_Right, padding=<ltrb 0 16 16 0>}}
  }
  Button_CloseTouch {
   widgetClassPath=/FortUILibrary/Foundation/Buttons/CloseButton
   bInherited=true
   bIsVariable=true
   parent=VisibilityBorder_Touch
   widget=Button_CloseTouch
   properties {
    onButtonBaseClicked=(&self:WidgetTree.Button_CloseTouch.HandleOnButtonClicked)
    onButtonBaseDoubleClicked=(&self:WidgetTree.Button_CloseTouch.HandleOnButtonDoubleClicked)
    onButtonBaseHovered=(&self:WidgetTree.Button_CloseTouch.HandleOnButtonHovered)
    onButtonBaseUnhovered=(&self:WidgetTree.Button_CloseTouch.HandleOnButtonUnhovered)
    onSelectedChangedBase=(&self:WidgetTree.Button_CloseTouch.HandleOnSelectedChanged)
   }
   slot {class=BorderSlot, properties {padding=<ltrb 0 0 0 0>}}
  }
 ]
}
