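# Sound cue for the UI sound played when the matchmaking Play button is pressed, the Creative variant 02 by name.
# A single wave player plays the shared PlayButton_Pressed_01 wave at reduced volume, routed to the critical in-game UI sound class.
# It joins tray-notification, matchmaking play-button and pressed-button concurrency groups, with a PreventNew override rule.
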SoundCue /Game/Sounds/Fort_UI_Sounds/FrontEnd/SC_UI_MatchMaking_PlayButton_Pressed_Creative_02
root=WavePlayer_0
nodes [{id=WavePlayer_0, properties {soundWaveAssetPtr=&/UI_MatchMaking_PlayButton_Pressed_01}}]
properties {
 maxDistance=2097152
 soundClassObject=&Fort_UI_InGame_Critical
 volumeMultiplier=0.7
 concurrencyOverrides {resolutionRule=PreventNew}
 concurrencySet [
  /Game/Sounds/Fort_UI_Sounds/TrayNotifications/CONC_TrayNotification
  &/CONC_MatchMaking_PlayButton
  /Game/Sounds/Fort_UI_Sounds/FrontEnd/DefaultSounds/PressedButtons/UI_PressedButton_Concurrency
 ]
}
